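# Control Rig for the F_MED_CoralNumber female character body, layering outfit and body corrective bone adjustments onto the animated pose.
# A shared DeformRig function runs first, then sequenced branches offset calf, armpit, belt, pants cuff, neck, glute and scapular mask bones.
# Remapped relative bone transforms and spherical pose readers weight the offsets, while parent and aim constraints steer tool, belt and mask bones.
# Skydive and front-end idle flags gate extra fixes.
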ControlRigBlueprint /BRCosmetics/Characters/Player/Female/Medium/Bodies/F_MED_CoralNumber/Meshes/F_MED_CoralNumber_CtrlRig
graphs [
 New Function_ContainedGraph {nodes [
   n1 = RigVMFunctionEntryNode()
   n2 = &RigVMFunction_MathDoubleMul(
    B=&RigVMFunction_MathDoubleDiv(A=&RigVMFunction_MathDoubleClamp(Maximum=@n1.Max Angle, Minimum=@n1.Min Angle, Value=@n1.Value).Result, B=@n1.SearchAngle).Result
    A=/Script/RigVM.RigVMFunction_MathDoubleConstHalfPi() {is=struct RigVMUnitNode, template="HalfPi::Execute(out Value)"}.Value
   )
   > RigVMFunctionReturnNode(Result=/Script/RigVM.RigVMFunction_MathDoublePow(
     B=@n1.PowerOut
     A=/Script/RigVM.RigVMFunction_MathDoubleSin(Value=@n2.Result) {is=struct RigVMUnitNode, template="Sin::Execute(in Value,out Result)"}.Result
    ) {is=struct RigVMUnitNode, template="Power::Execute(in A,in B,out Result)"}.Result)
  ]}
 RigVMFunctionLibrary {nodes [gentle peak = RigVMCollapseNode(Max Angle=90, Min Angle=-90, PowerOut=1, SearchAngle=0, Value=0)]}
 RigVMFunction_Sequence_1_ContainedGraph {nodes [
   n1 = RigVMFunctionEntryNode()
   RigVMFunction_Sequence = &RigVMFunction_Sequence(ExecuteContext=@n1.ExecuteContext)
   n2 = &RigVMFunction_Sequence(ExecuteContext=@RigVMFunction_Sequence.B)
   n3 = &RigVMFunction_Sequence(ExecuteContext=@n2.B)
   n4 = &RigVMFunction_Sequence(ExecuteContext=@n3.B)
   n5 = &RigVMFunction_Sequence(ExecuteContext=@n4.B)
   n6 = &RigVMFunction_Sequence(ExecuteContext=@n5.B)
   n7 = &RigVMFunction_Sequence(ExecuteContext=@n6.B)
   RigVMFunctionReturnNode(A=@RigVMFunction_Sequence.A, B=@n2.A, C=@n3.A, D=@n4.A, E=@n5.A, F=@n6.A, G=@n7.A, H=@n7.B)
  ]}
 RigVMModel {nodes [
   RigUnit_BeginExecution = &RigUnit_BeginExecution()
   FN_DeformRig_FNC = RigVMFunctionReferenceNode(
    DeformRig_ Curve="(deformRig_l_upperarm_up_alpha_72_#1=False,deformRig_r_upperarm_up_alpha_73_#2=False,deformRig_l_upperarm_fr_alpha_74_#3=False,deformRig_r_upperarm_fr_alpha_75_CB51B8 ...(truncated)"
    DeformRig_Bool="(upperarm_1_#4=True,elbow_3_#5=True,wrist_7_#6=True,groin_8_#7=true,glute_17_#8=true,knee_9_40C408324CA261 ...(truncated)"
    DeformRig_Debug=(upperarm_debug_18_#4=False,wrist_debug_20_#6=False,lat_debug_26_#9=False,groinGlute_debug_27_#7=False)
    DeformRig_Value="(upperarm_up_18_#10=(Rotation=(X=0,Y=0,Z=0,W=1),Translation=(X=0,Y=0,Z=5),Scale3D=(X=0.7,Y=1,Z=1)),upperarm_fwd_fr_19_#11=(Rotation=(X=0,Y=0,Z=0,W=1),Translation=(X=-2,Y=15,Z=5),Scale3D=(X=1,Y=1,Z=1 ...(truncated)"
    ExecuteContext=@RigUnit_BeginExecution.ExecutePin
   )
   n1 = &RigVMFunction_Sequence(ExecuteContext=@FN_DeformRig_FNC.ExecuteContext)
   &RigUnit_ModifyTransforms(
    ItemToModify=((Item=(Type=Bone,Name="calf_twist_01_r"),Transform=(Rotation=(X=0,Y=0,Z=0,W=1),Translation=(X=-4.2,Y=0,Z=0),Scale3D=(X=1,Y=1,Z=1))))
    WeightMinimum=0
    Weight=&RigVMFunction_MathDoubleRemap(
     SourceMaximum=42.5
     SourceMinimum=46.7
     TargetMaximum=1
     TargetMinimum=0
     bClamp=False
     Value=&RigUnit_GetRelativeTransformForItem(Child=(Type=Bone,Name="foot_r"), Parent=(Type=Bone,Name="calf_r"), bChildInitial=False, bParentInitial=False).RelativeTransform.Translation.X
    ).Result
    ExecutePin=@n1.A
   )
   > &RigUnit_ModifyTransforms(
    ItemToModify=((Item=(Type=Bone,Name="calf_twist_01_l"),Transform=(Rotation=(X=0,Y=0,Z=0,W=1),Translation=(X=4.2,Y=0,Z=0),Scale3D=(X=1,Y=1,Z=1))))
    WeightMinimum=0
    Weight=&RigVMFunction_MathDoubleRemap(
     SourceMaximum=-42.5
     SourceMinimum=-46.7
     TargetMaximum=1
     TargetMinimum=0
     bClamp=False
     Value=&RigUnit_GetRelativeTransformForItem(Child=(Type=Bone,Name="foot_l"), Parent=(Type=Bone,Name="calf_l"), bChildInitial=False, bParentInitial=False).RelativeTransform.Translation.X
    ).Result
   )
   n2 = &RigUnit_GetRelativeTransformForItem(Child=(Type=Bone,Name="upperarm_r"), Parent=(Type=Null,Name="upperarm_r_DefaultPose"), bChildInitial=False, bParentInitial=False)
   n3 = &RigVMFunction_MathQuaternionToEuler(RotationOrder=ZYX, Value=@n2.RelativeTransform.Rotation)
   Multiply = &RigVMFunction_MathDoubleMul(
    A=&RigVMFunction_MathDoubleRemap(SourceMaximum=35, SourceMinimum=10, TargetMaximum=1, TargetMinimum=0, bClamp=False, Value=@n3.Result.Y).Result
    B=&RigVMFunction_MathDoubleRemap(SourceMaximum=100, SourceMinimum=35, TargetMaximum=1, TargetMinimum=0, bClamp=False, Value=@n3.Result.Z).Result
   )
   > &RigUnit_ModifyTransforms(
    ItemToModify=((Item=(Type=Bone,Name="Mstr_edge_r"),Transform=(Rotation=(X=0.0421,Y=-0.0113,Z=-0.2586,W=0.965),Translation=(X=3.4054,Y=-9.7883,Z=5.268),Scale3D=(X=1,Y=1,Z=1))))
    WeightMinimum=0
    Weight=@Multiply.Result
   )
   > n4 = &RigUnit_SphericalPoseReader(
    ActiveRegionScaleFactors=(PositiveWidth=1,NegativeWidth=1,PositiveHeight=1,NegativeHeight=1)
    ActiveRegionSize=0.1
    DriverAxis=(X=-1,Y=0,Z=0)
    DriverItem=(Type=Bone,Name="upperarm_r")
    FalloffRegionScaleFactors=(PositiveWidth=1,NegativeWidth=1,PositiveHeight=1,NegativeHeight=1)
    FalloffSize=0.2
    FlipHeightScaling=False
    FlipWidthScaling=False
    OptionalParentItem=(Type=Bone,Name="spine_05")
    RotationOffset=(X=180,Y=-10,Z=0)
    Debug=RigVMVariableNode() {variable="Debug Drow "}.Value
   )
   > &RigUnit_ModifyTransforms(
    ItemToModify=((Item=(Type=Bone,Name="Mstr_edge_r"),Transform=(Rotation=(X=0,Y=0,Z=0,W=1),Translation=(X=3.1834,Y=0,Z=-3.1875),Scale3D=(X=1,Y=1,Z=1))))
    WeightMinimum=0
    Weight=@n4.OutputParam
   )
   > n5 = &RigUnit_SphericalPoseReader(
    ActiveRegionScaleFactors=(PositiveWidth=1,NegativeWidth=1,PositiveHeight=1,NegativeHeight=1)
    ActiveRegionSize=0.2
    DriverAxis=(X=1,Y=0,Z=0)
    DriverItem=(Type=Bone,Name="thigh_r")
    FalloffRegionScaleFactors=(PositiveWidth=1,NegativeWidth=1,PositiveHeight=1,NegativeHeight=1)
    FalloffSize=0.4
    FlipHeightScaling=False
    FlipWidthScaling=False
    OptionalParentItem=(Type=Bone,Name="None")
    RotationOffset=(X=-11.2221,Y=-4.8354,Z=12.4327)
    Debug=RigVMVariableNode() {variable="Debug Drow "}.Value
   )
   > &RigUnit_ModifyTransforms(
    ItemToModify=((Item=(Type=Bone,Name="thigh_twist_01_r"),Transform=(Rotation=(X=0.2666,Y=-0.005,Z=-0.0631,W=0.9617),Translation=(X=0,Y=0,Z=0),Scale3D=(X=1,Y=1,Z=1))))
    WeightMinimum=0
    Weight=@n5.OutputParam
   )
   n6 = RigVMVariableNode() {variable="Debug Drow "}
   > n7 = &RigUnit_SphericalPoseReader(
    ActiveRegionScaleFactors=(PositiveWidth=1,NegativeWidth=1,PositiveHeight=1,NegativeHeight=1)
    ActiveRegionSize=0.1
    DriverAxis=(X=-1,Y=0,Z=0)
    DriverItem=(Type=Bone,Name="upperarm_r")
    FalloffRegionScaleFactors=(PositiveWidth=1,NegativeWidth=1,PositiveHeight=1,NegativeHeight=1)
    FalloffSize=0.3
    FlipHeightScaling=False
    FlipWidthScaling=False
    OptionalParentItem=(Type=Bone,Name="spine_05")
    RotationOffset=(X=0,Y=40,Z=0)
    Debug=@n6.Value
   )
   > &RigUnit_ModifyTransforms(
    ItemToModify=((Item=(Type=Bone,Name="Mstr_Jacket_armpit_r"),Transform=(Rotation=(X=0,Y=0,Z=0.568,W=0.8231),Translation=(X=-6.2404,Y=1.2382,Z=-1.6824),Scale3D=(X=1,Y=1,Z=1))))
    WeightMinimum=0
    Weight=@n7.OutputParam
   )
   > n8 = &RigUnit_SphericalPoseReader(
    ActiveRegionScaleFactors=(PositiveWidth=1,NegativeWidth=1,PositiveHeight=1,NegativeHeight=1)
    ActiveRegionSize=0.1
    DriverAxis=(X=1,Y=0,Z=0)
    DriverItem=(Type=Bone,Name="upperarm_l")
    FalloffRegionScaleFactors=(PositiveWidth=1,NegativeWidth=1,PositiveHeight=1,NegativeHeight=1)
    FalloffSize=0.3
    FlipHeightScaling=False
    FlipWidthScaling=False
    OptionalParentItem=(Type=Bone,Name="spine_05")
    RotationOffset=(X=0,Y=220,Z=0)
    Debug=@n6.Value
   )
   > &RigUnit_ModifyTransforms(
    ItemToModify=((Item=(Type=Bone,Name="Mstr_Jacket_armpit_l"),Transform=(Rotation=(X=0,Y=0,Z=0,W=1),Translation=(X=5.6431,Y=-12.6803,Z=-4.4541),Scale3D=(X=1,Y=1,Z=1))))
    WeightMinimum=0
    Weight=@n8.OutputParam
   )
   RigVMFunction_Sequence = RigVMAggregateNode(ExecuteContext=@n1.B)
   &RigUnit_ParentConstraint(
    AdvancedSettings=(InterpolationType=Average,RotationOrderForFilter=XZY)
    Child=(Type=Bone,Name="Mstr_screwdriver")
    Parents=((Item=(Type=Bone,Name="thigh_l"),Weight=1),(Item=(Type=Bone,Name="thigh_twist_01_l"),Weight=0.5))
    Weight=1
    ExecutePin=@RigVMFunction_Sequence.A
   )
   > &RigUnit_ParentConstraint(
    AdvancedSettings=(InterpolationType=Average,RotationOrderForFilter=XZY)
    Child=(Type=Bone,Name="Mstr_wrench")
    Parents=((Item=(Type=Bone,Name="thigh_l"),Weight=1),(Item=(Type=Bone,Name="thigh_twist_01_l"),Weight=0.5))
    Weight=1
   )
   n9 = &RigUnit_GetRelativeTransformForItem(Child=(Type=Bone,Name="thigh_l"), Parent=(Type=Null,Name="thigh_l_DefaultPos"), bChildInitial=False, bParentInitial=False)
   &RigUnit_ModifyTransforms(
    ItemToModify="((Item=(Type=Bone,Name=\"Mstr_belt_left\"),Transform=(Rotation=(X=-0.0283,Y=-0.2821,Z=-0.0686,W=0.9565),Translation=(X=-2.1194,Y=1.1769,Z=-3.3043),Scale3D=(X=0.693,Y=1,Z=1))),(Item=(Type=Bone,Name=\"Mstr_belt_mid\"),Transform=(Rotation=(X=0.1348,Y=-0.1457,Z=0. ...(truncated)"
    WeightMinimum=0
    Weight=&RigVMFunction_MathDoubleRemap(
     SourceMaximum=-90
     SourceMinimum=0
     TargetMaximum=1
     TargetMinimum=0
     bClamp=False
     Value=&RigVMFunction_MathQuaternionToEuler(RotationOrder=ZYX, Value=@n9.RelativeTransform.Rotation).Result.Z
    ).Result
    ExecutePin=@RigVMFunction_Sequence.B
   )
   n10 = &RigUnit_GetRelativeTransformForItem(Child=(Type=Bone,Name="thigh_r"), Parent=(Type=Null,Name="thigh_r_DefaultPos"), bChildInitial=False, bParentInitial=False)
   > &RigUnit_ModifyTransforms(
    ItemToModify="((Item=(Type=Bone,Name=\"Mstr_belt_right\"),Transform=(Rotation=(X=-0.2155,Y=0.0316,Z=-0.0601,W=0.9741),Translation=(X=-1.3027,Y=0.0048,Z=2.0011),Scale3D=(X=1,Y=1,Z=1))),(Item=(Type=Bone,Name=\"Mstr_belt_mid\"),Transform=(Rotation=(X=0,Y=0.0427,Z=0,W=0.9991),T ...(truncated)"
    WeightMinimum=0
    Weight=&RigVMFunction_MathDoubleRemap(
     SourceMaximum=-85
     SourceMinimum=-65
     TargetMaximum=1
     TargetMinimum=0
     bClamp=False
     Value=&RigVMFunction_MathQuaternionToEuler(RotationOrder=ZYX, Value=@n10.RelativeTransform.Rotation).Result.Z
    ).Result
   )
   n11 = &RigUnit_GetRelativeTransformForItem(Child=(Type=Bone,Name="thigh_r"), Parent=(Type=Null,Name="thigh_r_DefaultPos"), bChildInitial=False, bParentInitial=False)
   &RigUnit_ModifyTransforms(
    ItemToModify="((Item=(Type=Bone,Name=\"Mstr_belt_right\"),Transform=(Rotation=(X=0.3611,Y=-0.1702,Z=0.3639,W=0.8416),Translation=(X=-1.9921,Y=1.8004,Z=2.4661),Scale3D=(X=1,Y=1,Z=1))),(Item=(Type=Bone,Name=\"Mstr_belt_mid\"),Transform=(Rotation=(X=0.1735,Y=0.2828,Z=-0.0829,W ...(truncated)"
    WeightMinimum=0
    Weight=&RigVMFunction_MathDoubleRemap(
     SourceMaximum=-91
     SourceMinimum=-8
     TargetMaximum=1
     TargetMinimum=0
     bClamp=False
     Value=&RigVMFunction_MathQuaternionToEuler(RotationOrder=ZYX, Value=@n11.RelativeTransform.Rotation).Result.Z
    ).Result
    ExecutePin=@RigVMFunction_Sequence.C
   )
   n12 = &RigVMFunction_MathQuaternionToEuler(
    RotationOrder=ZYX
    Value=&RigUnit_GetRelativeTransformForItem(Child=(Type=Bone,Name="foot_r"), Parent=(Type=Null,Name="foot_r_DefaultPos"), bChildInitial=False, bParentInitial=False).RelativeTransform.Rotation
   )
   > &RigUnit_ModifyTransforms(
    ItemToModify="((Item=(Type=Bone,Name=\"Mstr_pants_cuffs_outter_r\"),Transform=(Rotation=(X=-0.0347,Y=0.0881,Z=0.1527,W=0.9837),Translation=(X=2.1095,Y=1.0399,Z=2.2935),Scale3D=(X=1,Y=1,Z=1))),(Item=(Type=Bone,Name=\"Mstr_pants_cuffs_Inner_r\"),Transform=(Rotation=(X=0.2309, ...(truncated)"
    WeightMinimum=0
    Weight=&RigVMFunction_MathDoubleRemap(SourceMaximum=-30, SourceMinimum=0, TargetMaximum=1, TargetMinimum=0, bClamp=False, Value=@n12.Result.Z).Result
   )
   n13 = &RigVMFunction_MathDoubleRemap(SourceMaximum=70, SourceMinimum=0, TargetMaximum=1, TargetMinimum=0, bClamp=False, Value=@n12.Result.Z)
   > &RigUnit_ModifyTransforms(
    ItemToModify="((Item=(Type=Bone,Name=\"Mstr_pants_cuffs_outter_r\"),Transform=(Rotation=(X=0.087,Y=0.0367,Z=-0.0827,W=0.9921),Translation=(X=0.1662,Y=0.9489,Z=0.1627),Scale3D=(X=1,Y=1,Z=1))),(Item=(Type=Bone,Name=\"Mstr_pants_cuffs_Inner_r\"),Transform=(Rotation=(X=-0.1298, ...(truncated)"
    WeightMinimum=0
    Weight=@n13.Result
   )
   > &RigUnit_ModifyTransforms(
    ItemToModify=((Item=(Type=Bone,Name="calf_twist_01_r"),Transform=(Rotation=(X=0,Y=0,Z=0,W=1),Translation=(X=-2,Y=0,Z=0),Scale3D=(X=1,Y=1,Z=1))))
    WeightMinimum=0
    Weight=@n13.Result
   )
   > &RigUnit_ModifyTransforms(
    ItemToModify="((Item=(Type=Bone,Name=\"Mstr_pants_cuffs_outter_r\"),Transform=(Rotation=(X=0.0907,Y=0.0079,Z=0.317,W=0.944),Translation=(X=0.1694,Y=2.5962,Z=0),Scale3D=(X=1,Y=1,Z=1))),(Item=(Type=Bone,Name=\"Mstr_pants_cuffs_Inner_r\"),Transform=(Rotation=(X=-0.0255,Y=0.089 ...(truncated)"
    WeightMinimum=0
    Weight=&RigVMFunction_MathDoubleRemap(SourceMaximum=-30, SourceMinimum=0, TargetMaximum=1, TargetMinimum=0, bClamp=False, Value=@n12.Result.X).Result
   )
   > &RigUnit_ModifyTransforms(
    ItemToModify="((Item=(Type=Bone,Name=\"Mstr_pants_cuffs_outter_r\"),Transform=(Rotation=(X=-0.0093,Y=0.0754,Z=-0.0124,W=0.997),Translation=(X=0.7866,Y=-0.1622,Z=-0.2307),Scale3D=(X=1,Y=1,Z=1))),(Item=(Type=Bone,Name=\"Mstr_pants_cuffs_Inner_r\"),Transform=(Rotation=(X=-0.02 ...(truncated)"
    WeightMinimum=0
    Weight=&RigVMFunction_MathDoubleRemap(SourceMaximum=30, SourceMinimum=0, TargetMaximum=1, TargetMinimum=0, bClamp=False, Value=@n12.Result.X).Result
   )
   > &RigUnit_ModifyTransforms(
    ItemToModify="((Item=(Type=Bone,Name=\"Mstr_pants_cuffs_outter_r\"),Transform=(Rotation=(X=-0.0146,Y=-0.063,Z=0.1168,W=0.991),Translation=(X=0.2312,Y=0.0559,Z=0.0291),Scale3D=(X=1,Y=1,Z=1))),(Item=(Type=Bone,Name=\"Mstr_pants_cuffs_Inner_r\"),Transform=(Rotation=(X=0,Y=0,Z= ...(truncated)"
    WeightMinimum=0
    Weight=&RigVMFunction_MathDoubleRemap(SourceMaximum=-30, SourceMinimum=0, TargetMaximum=1, TargetMinimum=0, bClamp=False, Value=@n12.Result.Y).Result
   )
   > &RigUnit_ModifyTransforms(
    ItemToModify="((Item=(Type=Bone,Name=\"Mstr_pants_cuffs_outter_r\"),Transform=(Rotation=(X=0,Y=0,Z=0,W=1),Translation=(X=0,Y=0,Z=0),Scale3D=(X=1,Y=1,Z=1))),(Item=(Type=Bone,Name=\"Mstr_pants_cuffs_Inner_r\"),Transform=(Rotation=(X=0.005,Y=-0.0581,Z=-0.0959,W=0.9937),Transla ...(truncated)"
    WeightMinimum=0
    Weight=&RigVMFunction_MathDoubleRemap(SourceMaximum=30, SourceMinimum=0, TargetMaximum=1, TargetMinimum=0, bClamp=False, Value=@n12.Result.Y).Result
   )
   n14 = &RigVMFunction_MathQuaternionToEuler(
    RotationOrder=ZYX
    Value=&RigUnit_GetRelativeTransformForItem(Child=(Type=Bone,Name="foot_l"), Parent=(Type=Null,Name="foot_l_DefaultPos"), bChildInitial=False, bParentInitial=False).RelativeTransform.Rotation
   )
   > &RigUnit_ModifyTransforms(
    ItemToModify="((Item=(Type=Bone,Name=\"Mstr_pants_cuffs_outter_l\"),Transform=(Rotation=(X=0,Y=0,Z=0.1413,W=0.99),Translation=(X=-0.7216,Y=0,Z=0),Scale3D=(X=0.7892,Y=1,Z=1))),(Item=(Type=Bone,Name=\"Mstr_pants_cuffs_Inner_l\"),Transform=(Rotation=(X=0.1265,Y=-0.0109,Z=-0.08 ...(truncated)"
    WeightMinimum=0
    Weight=&RigVMFunction_MathDoubleRemap(SourceMaximum=-30, SourceMinimum=0, TargetMaximum=1, TargetMinimum=0, bClamp=False, Value=@n14.Result.Z).Result
   )
   n15 = &RigVMFunction_MathDoubleRemap(SourceMaximum=70, SourceMinimum=0, TargetMaximum=1, TargetMinimum=0, bClamp=False, Value=@n14.Result.Z)
   > &RigUnit_ModifyTransforms(
    ItemToModify="((Item=(Type=Bone,Name=\"Mstr_pants_cuffs_outter_l\"),Transform=(Rotation=(X=0.0798,Y=0.0042,Z=-0.1743,W=0.9814),Translation=(X=0,Y=0.0371,Z=-1.9795),Scale3D=(X=1,Y=1,Z=1))),(Item=(Type=Bone,Name=\"Mstr_pants_cuffs_Inner_l\"),Transform=(Rotation=(X=-0.068,Y=0, ...(truncated)"
    WeightMinimum=0
    Weight=@n15.Result
   )
   > &RigUnit_ModifyTransforms(
    ItemToModify=((Item=(Type=Bone,Name="calf_twist_01_l"),Transform=(Rotation=(X=0,Y=0,Z=0,W=1),Translation=(X=2,Y=0,Z=0),Scale3D=(X=1,Y=1,Z=1))))
    WeightMinimum=0
    Weight=@n15.Result
   )
   > &RigUnit_ModifyTransforms(
    ItemToModify="((Item=(Type=Bone,Name=\"Mstr_pants_cuffs_outter_l\"),Transform=(Rotation=(X=0.051,Y=-0.0178,Z=0.2259,W=0.9727),Translation=(X=-0.7467,Y=-2.004,Z=0),Scale3D=(X=1,Y=1,Z=1))),(Item=(Type=Bone,Name=\"Mstr_pants_cuffs_Inner_l\"),Transform=(Rotation=(X=0.0101,Y=0.0 ...(truncated)"
    WeightMinimum=0
    Weight=&RigVMFunction_MathDoubleRemap(SourceMaximum=-30, SourceMinimum=0, TargetMaximum=1, TargetMinimum=0, bClamp=False, Value=@n14.Result.X).Result
   )
   > &RigUnit_ModifyTransforms(
    ItemToModify="((Item=(Type=Bone,Name=\"Mstr_pants_cuffs_outter_l\"),Transform=(Rotation=(X=-0.0079,Y=0.0919,Z=-0.0783,W=0.9927),Translation=(X=0,Y=-0.475,Z=0),Scale3D=(X=1,Y=1,Z=1))),(Item=(Type=Bone,Name=\"Mstr_pants_cuffs_Inner_l\"),Transform=(Rotation=(X=-0.0147,Y=-0.107 ...(truncated)"
    WeightMinimum=0
    Weight=&RigVMFunction_MathDoubleRemap(SourceMaximum=30, SourceMinimum=0, TargetMaximum=1, TargetMinimum=0, bClamp=False, Value=@n14.Result.X).Result
   )
   > &RigUnit_ModifyTransforms(
    ItemToModify="((Item=(Type=Bone,Name=\"Mstr_pants_cuffs_outter_l\"),Transform=(Rotation=(X=0.0764,Y=-0.0918,Z=0.0055,W=0.9928),Translation=(X=0,Y=0,Z=0),Scale3D=(X=0.9649,Y=1,Z=1))),(Item=(Type=Bone,Name=\"Mstr_pants_cuffs_Inner_l\"),Transform=(Rotation=(X=0,Y=0,Z=0,W=1),Tr ...(truncated)"
    WeightMinimum=0
    Weight=&RigVMFunction_MathDoubleRemap(SourceMaximum=-30, SourceMinimum=0, TargetMaximum=1, TargetMinimum=0, bClamp=False, Value=@n14.Result.Y).Result
   )
   > &RigUnit_ModifyTransforms(
    ItemToModify="((Item=(Type=Bone,Name=\"Mstr_pants_cuffs_outter_l\"),Transform=(Rotation=(X=0,Y=0,Z=0,W=1),Translation=(X=0,Y=0,Z=0),Scale3D=(X=1,Y=1,Z=1))),(Item=(Type=Bone,Name=\"Mstr_pants_cuffs_Inner_l\"),Transform=(Rotation=(X=0.0069,Y=-0.0781,Z=-0.1255,W=0.989),Transla ...(truncated)"
    WeightMinimum=0
    Weight=&RigVMFunction_MathDoubleRemap(SourceMaximum=30, SourceMinimum=0, TargetMaximum=1, TargetMinimum=0, bClamp=False, Value=@n14.Result.Y).Result
   )
   n16 = &RigUnit_GetRelativeTransformForItem(Child=(Type=Bone,Name="head"), Parent=(Type=Bone,Name="spine_05"), bChildInitial=False, bParentInitial=False)
   &RigUnit_ModifyTransforms(
    ItemToModify=((Item=(Type=Bone,Name="Mstr_BackNeck"),Transform=(Rotation=(X=0.4232,Y=-0.0661,Z=-0.0323,W=0.9031),Translation=(X=-0.6843,Y=-2.3405,Z=-4.1837),Scale3D=(X=1,Y=1,Z=1))))
    WeightMinimum=0
    Weight=&RigVMFunction_MathDoubleRemap(
     SourceMaximum=-60
     SourceMinimum=-40
     TargetMaximum=1
     TargetMinimum=0
     bClamp=False
     Value=&RigVMFunction_MathQuaternionToEuler(RotationOrder=ZYX, Value=@n16.RelativeTransform.Rotation).Result.Z
    ).Result
    ExecutePin=@RigVMFunction_Sequence.E
   )
   &RigUnit_ModifyTransforms(
    ItemToModify=((Item=(Type=Bone,Name="thigh_l"),Transform=(Rotation=(X=0,Y=0,Z=0,W=1),Translation=(X=0,Y=0,Z=0),Scale3D=(X=1,Y=1,Z=1))))
    Weight=1
    WeightMinimum=0
    ExecutePin=@RigVMFunction_Sequence.F
   )
   > &RigUnit_AimConstraintLocalSpaceOffset(
    AdvancedSettings=(DebugSettings=(bEnabled=true,Scale=10,WorldOffset=(Rotation=(X=0,Y=0,Z=0,W=1),Translation=(X=0,Y=0,Z=0),Scale3D=(X=1,Y=1,Z=1))),RotationOrderForFilter=XZY)
    Child=(Type=Bone,Name="Mstr_belt_back")
    Filter=(bX=True,bY=True,bZ=true)
    Parents=((Item=(Type=Bone,Name="calf_l"),Weight=1))
    Weight=1
    WorldUp=(Target=(X=1,Y=0,Z=0),Kind=Location,Space=(Type=Bone,Name="thigh_r"))
   )
   n17 = &RigUnit_GetRelativeTransformForItem(Child=(Type=Bone,Name="thigh_l"), Parent=(Type=Null,Name="thigh_l_DefaultPos"), bChildInitial=False, bParentInitial=False)
   n18 = &RigVMFunction_MathQuaternionToEuler(RotationOrder=ZYX, Value=@n17.RelativeTransform.Rotation)
   > &RigUnit_ParentConstraint(
    AdvancedSettings=(InterpolationType=Average,RotationOrderForFilter=XZY)
    Child=(Type=Bone,Name="Mstr_belt_back_glute")
    Filter=(TranslationFilter=(bX=true,bY=True,bZ=true),RotationFilter=(bX=false,bY=false,bZ=false),ScaleFilter=(bX=false,bY=false,bZ=false))
    Parents=((Item=(Type=Bone,Name="deform_glute_l"),Weight=1))
    Weight=&RigVMFunction_MathDoubleRemap(SourceMaximum=-30, SourceMinimum=0, TargetMaximum=1, TargetMinimum=0, Value=@n18.Result.Z).Result
   )
   > &RigUnit_ModifyTransforms(
    ItemToModify=((Item=(Type=Bone,Name="Mstr_belt_back_glute"),Transform=(Rotation=(X=-0.3948,Y=0.0097,Z=0.1378,W=0.9083),Translation=(X=-2.9359,Y=1.0086,Z=-0.2297),Scale3D=(X=1.1415,Y=1,Z=1))))
    WeightMinimum=0
    Weight=&RigVMFunction_MathDoubleRemap(SourceMaximum=-60, SourceMinimum=0, TargetMaximum=1, TargetMinimum=0, Value=@n18.Result.Z).Result
   )
   > &RigUnit_ModifyTransforms(
    ItemToModify="((Item=(Type=Bone,Name=\"Mstr_belt_back_glute\"),Transform=(Rotation=(X=-0.1459,Y=0.0947,Z=0.0319,W=0.9842),Translation=(X=0.445,Y=-0.8386,Z=-0.0099),Scale3D=(X=1.1678,Y=1,Z=1))),(Item=(Type=Bone,Name=\"Mstr_belt_back\"),Transform=(Rotation=(X=-0.0039,Y=0.0112 ...(truncated)"
    WeightMinimum=0
    Weight=&RigVMFunction_MathDoubleRemap(SourceMaximum=-55, SourceMinimum=0, TargetMaximum=1, TargetMinimum=0, Value=@n18.Result.X).Result
   )
   n19 = &RigVMFunction_MathQuaternionToEuler(RotationOrder=ZYX, Value=@n17.RelativeTransform.Rotation)
   > &RigUnit_ModifyTransforms(
    ItemToModify="((Item=(Type=Bone,Name=\"Mstr_belt_back_glute\"),Transform=(Rotation=(X=0.4266,Y=0,Z=0,W=0.9045),Translation=(X=0,Y=1.0529,Z=2.6891),Scale3D=(X=1,Y=1,Z=1))),(Item=(Type=Bone,Name=\"Mstr_belt_back\"),Transform=(Rotation=(X=0.5432,Y=-0.5241,Z=0.2274,W=0.6153),Tr ...(truncated)"
    WeightMinimum=0
    Weight=&RigVMFunction_MathDoubleMul(
     B=&RigVMFunction_MathDoubleRemap(SourceMaximum=-87, SourceMinimum=-30, TargetMaximum=1, TargetMinimum=0, Value=@n19.Result.Z).Result
     A=&RigVMFunction_MathDoubleRemap(SourceMaximum=-73, SourceMinimum=-30, TargetMaximum=1, TargetMinimum=0, Value=@n19.Result.X).Result
    ).Result
   )
   RigUnit_GetRelativeTransformForItem = &RigUnit_GetRelativeTransformForItem(Child=(Type=Bone,Name="calf_l"), Parent=(Type=Bone,Name="thigh_l"), bChildInitial=False, bParentInitial=False)
   RigVMFunction_MathQuaternionToEuler = &RigVMFunction_MathQuaternionToEuler(RotationOrder=ZYX, Value=@RigUnit_GetRelativeTransformForItem.RelativeTransform.Rotation)
   Remap = &RigVMFunction_MathDoubleRemap(SourceMaximum=120, SourceMinimum=80, TargetMaximum=1, TargetMinimum=0, Value=@RigVMFunction_MathQuaternionToEuler.Result.Z)
   n20 = &RigVMFunction_MathQuaternionToEuler(
    RotationOrder=ZYX
    Value=&RigUnit_GetRelativeTransformForItem(Child=(Type=Bone,Name="calf_r"), Parent=(Type=Bone,Name="thigh_r"), bChildInitial=False, bParentInitial=False).RelativeTransform.Rotation
   )
   RigVMVariableNode(
    Value=&RigVMFunction_MathDoubleMul(
     A=@Remap.Result
     B=&RigVMFunction_MathDoubleRemap(SourceMaximum=120, SourceMinimum=80, TargetMaximum=1, TargetMinimum=0, Value=@n20.Result.Z).Result
    ).Result
    ExecuteContext=@RigVMFunction_Sequence.G
   ) {variable="Calfs Bend  Alpha"}
   n21 = &RigUnit_GetRelativeTransformForItem(Child=(Type=Bone,Name="thigh_l"), Parent=(Type=Null,Name="thigh_l_DefaultPos"), bChildInitial=False, bParentInitial=False)
   n22 = RigVMVariableNode() {variable="Calfs Bend  Alpha"}
   > &RigUnit_ModifyTransforms(
    ItemToModify="((Item=(Type=Bone,Name=\"deform_glute_base_l\"),Transform=(Rotation=(X=0,Y=0,Z=0,W=1),Translation=(X=-2.9983,Y=0.1451,Z=1.8497),Scale3D=(X=1,Y=1,Z=1))),(Item=(Type=Bone,Name=\"Mstr_belt_back_glute\"),Transform=(Rotation=(X=0.0741,Y=-0.0022,Z=-0.0291,W=0.9968), ...(truncated)"
    WeightMinimum=0
    Weight=&RigVMFunction_MathDoubleRemap(
     SourceMaximum=-70
     SourceMinimum=-30
     TargetMinimum=0
     Value=&RigVMFunction_MathQuaternionToEuler(RotationOrder=ZYX, Value=@n21.RelativeTransform.Rotation).Result.Z
     TargetMaximum=@n22.Value
    ).Result
   )
   n23 = &RigUnit_GetRelativeTransformForItem(Child=(Type=Bone,Name="thigh_twist_01_r"), Parent=(Type=Null,Name="thigh_r_DefaultPos"), bChildInitial=False, bParentInitial=False)
   > n24 = &RigUnit_ModifyTransforms(
    ItemToModify=((Item=(Type=Bone,Name="deform_glute_base_r"),Transform=(Rotation=(X=0,Y=0,Z=-0.0186,W=0.9998),Translation=(X=-2.5256,Y=-1.4608,Z=-0.7818),Scale3D=(X=0.6687,Y=1,Z=1))))
    WeightMinimum=0
    Weight=&RigVMFunction_MathDoubleRemap(
     SourceMaximum=-70
     SourceMinimum=-30
     TargetMinimum=0
     Value=&RigVMFunction_MathQuaternionToEuler(RotationOrder=ZYX, Value=@n23.RelativeTransform.Rotation).Result.Z
     TargetMaximum=@n22.Value
    ).Result
   )
   n25 = &RigVMFunction_ControlFlowBranch(ExecuteContext=@n24.ExecutePin, Condition=RigVMVariableNode() {variable=FrontEnd Idle Bypass}.Value)
   n26 = &RigUnit_SphericalPoseReader(
    ActiveRegionScaleFactors=(PositiveWidth=1,NegativeWidth=1,PositiveHeight=1,NegativeHeight=1)
    ActiveRegionSize=0.05
    Debug=(bDrawDebug=True,bDraw2D=False,bDrawLocalAxes=False,DebugScale=25,DebugSegments=20,DebugThickness=0.25)
    DriverAxis=(X=1,Y=0,Z=0)
    DriverItem=(Type=Bone,Name="hand_l")
    FalloffRegionScaleFactors=(PositiveWidth=1,NegativeWidth=1,PositiveHeight=1,NegativeHeight=1)
    FalloffSize=0.05
    FlipHeightScaling=False
    FlipWidthScaling=False
    OptionalParentItem=(Type=Bone,Name="None")
    RotationOffset=(X=0,Y=-135,Z=5)
    ExecutePin=@n25.True
   )
   > &RigUnit_ModifyTransforms(
    ItemToModify=((Item=(Type=Bone,Name="upperarm_l"),Transform=(Rotation=(X=0,Y=0,Z=0.0262,W=0.9997),Translation=(X=0,Y=0,Z=0),Scale3D=(X=1,Y=1,Z=1))))
    WeightMinimum=0
    Weight=@n26.OutputParam
   )
   n27 = &RigUnit_SphericalPoseReader(
    ActiveRegionScaleFactors=(PositiveWidth=1,NegativeWidth=1,PositiveHeight=1,NegativeHeight=1)
    ActiveRegionSize=0.2
    DriverAxis=(X=1,Y=0,Z=0)
    DriverItem=(Type=Bone,Name="upperarm_l")
    FalloffRegionScaleFactors=(PositiveWidth=1,NegativeWidth=1,PositiveHeight=1,NegativeHeight=1)
    FalloffSize=0.5
    FlipHeightScaling=False
    FlipWidthScaling=False
    OptionalParentItem=(Type=Bone,Name="None")
    RotationOffset=(X=3.7314,Y=-21.2159,Z=5.9474)
    Debug=RigVMVariableNode() {variable="Debug Drow "}.Value
    ExecutePin=@RigVMFunction_Sequence.D
   )
   RigVMFunction_MathVectorFromFloat = /Script/RigVM.RigVMFunction_MathVectorFromFloat(Value=@n27.OutputParam) {is=struct RigVMUnitNode, template="Cast::Execute(in Value,out Result)"}
   RigVMFunction_MathVectorMake = /Script/RigVM.RigVMFunction_MathVectorMake(X=0, Y=0, Z=0) {is=struct RigVMUnitNode}
   > &RigUnit_AimConstraintLocalSpaceOffset(
    AdvancedSettings=(DebugSettings=(bEnabled=true,Scale=10,WorldOffset=(Rotation=(X=0,Y=0,Z=0,W=1),Translation=(X=0,Y=0,Z=0),Scale3D=(X=1,Y=1,Z=1))),RotationOrderForFilter=XZY)
    Child=(Type=Bone,Name="Mstr_scapularMask_base")
    Filter=(bX=True,bY=True,bZ=True)
    Parents=((Item=(Type=Bone,Name="lowerarm_l"),Weight=1))
    Weight=1
    WorldUp=(Target=(X=-20,Y=20,Z=20),Kind=Location,Space=(Type=Bone,Name="spine_05"))
    WorldUp.Target=/Script/RigVM.RigVMFunction_MathVectorRemap(
     SourceMaximum=(X=1,Y=1,Z=1)
     SourceMinimum=(X=0,Y=0,Z=0)
     bClamp=False
     Value=@RigVMFunction_MathVectorFromFloat.Result
     TargetMinimum=@RigVMFunction_MathVectorMake.Result
     TargetMaximum=&RigVMFunction_MathVectorMake(X=-20, Y=30, Z=20).Result
    ) {
     is=struct RigVMUnitNode
     template="Remap::Execute(in Value,in SourceMinimum,in SourceMaximum,in TargetMinimum,in TargetMaximum,in bClamp,out Result)"
    }.Result
   )
   n28 = &RigUnit_GetRelativeTransformForItem(Child=(Type=Bone,Name="upperarm_l"), Parent=(Type=Null,Name="upperarm_l_DefaultPose"), bChildInitial=False, bParentInitial=False)
   n29 = &RigVMFunction_MathDoubleRemap(
    SourceMaximum=80
    SourceMinimum=0
    TargetMaximum=1
    TargetMinimum=0
    Value=&RigVMFunction_MathQuaternionToEuler(RotationOrder=ZYX, Value=@n28.RelativeTransform.Rotation).Result.X
   )
   > n30 = &RigUnit_ModifyTransforms(
    ItemToModify="((Item=(Type=Bone,Name=\"Mstr_scapularMask_base\"),Transform=(Rotation=(X=0,Y=0,Z=0,W=1),Translation=(X=4.5,Y=0,Z=0),Scale3D=(X=1,Y=1,Z=1))),(Item=(Type=Bone,Name=\"Mstr_scapularMask\"),Transform=(Rotation=(X=0,Y=0,Z=0,W=1),Translation=(X=0,Y=0,Z=0),Scale3D=(X ...(truncated)"
    WeightMinimum=0
    Weight=@n29.Result
   )
   n31 = RigVMVariableNode() {variable=skyDive}
   RigVMFunction_ControlFlowBranch = &RigVMFunction_ControlFlowBranch(ExecuteContext=@n30.ExecutePin, Condition=@n31.Value)
   RigVMFunction_MathBoolToInteger = &RigVMFunction_MathBoolToInteger(Value=@n31.Value)
   n32 = RigVMDispatchNode(Values=(0,1), Index=@RigVMFunction_MathBoolToInteger.Result) {template="DISPATCH_RigVMDispatch_SelectInt32(in Index,in Values,out Result)"}
   &RigUnit_ModifyTransforms(
    ItemToModify="((Item=(Type=Bone,Name=\"Mstr_scapularMask_base\"),Transform=(Rotation=(X=0,Y=0,Z=0,W=1),Translation=(X=0,Y=0,Z=0),Scale3D=(X=1,Y=1,Z=1))),(Item=(Type=Bone,Name=\"Mstr_scapularMask\"),Transform=(Rotation=(X=0.0002,Y=-0.0147,Z=-0.0125,W=0.9998),Translation=(X=- ...(truncated)"
    WeightMinimum=0
    Weight=&RigVMFunction_MathDoubleMul(
     A=&RigVMFunction_KalmanFloat(BufferSize=6, Value=@n32.Result) {template="KalmanFilter::Execute(in Value,in BufferSize,out Result)"}.Result
     B=@n29.Result
    ).Result
    ExecutePin=@RigVMFunction_ControlFlowBranch.True
   )
   &RigUnit_ModifyTransforms(
    ExecutePin=()
    ItemToModify=((Item=(Type=Bone,Name="deform_glute_base_r"),Transform=(Rotation=(X=0,Y=0,Z=-0.0186,W=0.9998),Translation=(X=-2.5259,Y=-1.461,Z=-0.782),Scale3D=(X=1,Y=1,Z=1))))
    Weight=1
    WeightMinimum=0
   )
   &RigUnit_ModifyTransforms(
    ExecutePin=()
    ItemToModify="((Item=(Type=Bone,Name=\"Mstr_belt_back_glute\"),Transform=(Rotation=(X=-0.1154,Y=0.0668,Z=0.0962,W=0.9864),Translation=(X=-0.3771,Y=-1.1633,Z=-1.6591),Scale3D=(X=1.1678,Y=1,Z=1))),(Item=(Type=Bone,Name=\"Mstr_belt_back\"),Transform=(Rotation=(X=0,Y=0,Z=0,W=1) ...(truncated)"
    Weight=1
    WeightMinimum=0
   )
  ]}
]
hierarchy [
 root {loc=<0 0 0>, rot=<0 0 0 1>}
 pelvis {parent=root, loc=<0 -0.5301 98.0571>, rot=<0.0127 0.707 -0.0127 -0.707>}
 spine_01 {parent=pelvis, loc=<6.1208 -0.9328 0>, rot=<0 0 0.034 -0.9994>}
 spine_02 {parent=spine_01, loc=<6.5996 -0.2095 0>, rot=<0 0 0.0238 -0.9997>}
 spine_03 {parent=spine_02, loc=<6.4914 -0.1237 0>, rot=<0 0 0.0149 -0.9999>}
 spine_04 {parent=spine_03, loc=<6.9598 0.0794 0>, rot=<0 0 -0.0132 -0.9999>}
 spine_05 {parent=spine_04, loc=<6.9987 0 0>, rot=<0 0 -0.0015 -1>}
 attach_cape {parent=spine_05, loc=<9.3216 -8.4501 0>, rot=<0.0616 -0.7044 -0.0616 -0.7044>}
 clavicle_l {parent=spine_05, loc=<5.4831 -0.1053 -3.2946>, rot=<0.0465 -0.7289 -0.009 -0.6829>}
 upperarm_l {parent=clavicle_l, loc=<9.8658 -1.2136 -0.4105>, rot=<0.0164 -0.3245 -0.0269 -0.9454>}
 lowerarm_l {parent=upperarm_l, loc=<26.465 -0.8345 0.259>, rot=<0.0036 0.0161 -0.2298 -0.9731>}
 hand_l {parent=lowerarm_l, loc=<22.7355 -0.3203 -0.6105>, rot=<-0.6784 -0.0223 0.0735 -0.7307>}
 thumb_01_l {parent=hand_l, loc=<1.3118 -0.3191 -2.459>, rot=<0.6546 -0.4042 0.0664 -0.6354>}
 thumb_02_l {parent=thumb_01_l, loc=<5.9336 -0.0448 0.0401>, rot=<-0.0037 0.0213 0.298 -0.9543>}
 thumb_03_l {parent=thumb_02_l, loc=<2.2223 0.3682 -0.0907>, rot=<0.0118 -0.013 0.0066 -0.9998>}
 index_metacarpal_l {parent=hand_l, loc=<1.6535 0.0183 -2.96>, rot=<0.0158 -0.0124 -0.0173 -0.9996>}
 index_01_l {parent=index_metacarpal_l, loc=<8.3008 0.0357 -0.0524>, rot=<-0.0044 0.0485 0.0924 -0.9945>}
 index_02_l {parent=index_01_l, loc=<3.3587 -0.3004 0.0452>, rot=<-0.002 0.0027 0.3225 -0.9466>}
 index_03_l {parent=index_02_l, loc=<2.5578 -0.123 -0.0252>, rot=<-0.002 0.0067 0.2558 -0.9667>}
 middle_metacarpal_l {parent=hand_l, loc=<2.1856 -0.1725 -0.5984>, rot=<0.006 0.0168 -0.0329 -0.9993>}
 middle_01_l {parent=middle_metacarpal_l, loc=<7.3275 -0.2447 0.0129>, rot=<-0.0208 0.0042 0.0858 -0.9961>}
 middle_02_l {parent=middle_01_l, loc=<4.0123 0.1401 0.1175>, rot=<-0.0218 0.0258 0.3512 -0.9357>}
 middle_03_l {parent=middle_02_l, loc=<2.6077 0.0534 0.1373>, rot=<0.0082 0.0188 0.1708 -0.9851>}
 ring_metacarpal_l {parent=hand_l, loc=<2.2074 -0.3625 1.3638>, rot=<-0.0284 0.0722 -0.0458 -0.9959>}
 ring_01_l {parent=ring_metacarpal_l, loc=<7.0295 -0.6197 -0.1446>, rot=<-0.0489 -0.0252 0.1892 -0.9804>}
 ring_02_l {parent=ring_01_l, loc=<3.6174 0.4647 -0.1027>, rot=<0 0 0.2806 -0.9598>}
 ring_03_l {parent=ring_02_l, loc=<2.4152 0.114 0.0199>, rot=<0 0 0.1189 -0.9929>}
 pinky_metacarpal_l {parent=hand_l, loc=<1.7616 -0.9753 3.3814>, rot=<-0.0397 0.0814 -0.069 -0.9935>}
 pinky_01_l {parent=pinky_metacarpal_l, loc=<7.1709 -0.4831 -0.1094>, rot=<-0.0654 -0.0035 0.2242 -0.9723>}
 pinky_02_l {parent=pinky_01_l, loc=<2.8274 0.0893 -0.1043>, rot=<0.0037 0.0011 0.2854 -0.9584>}
 pinky_03_l {parent=pinky_02_l, loc=<2.2046 0.0307 -0.0078>, rot=<-0.0002 0.0098 0.1445 -0.9895>}
 deform_wrist_dn_base_l {parent=hand_l, loc=<0 -1.5326 0>, rot=<0 0 0 -1>}
 deform_wrist_dn_l {parent=deform_wrist_dn_base_l, loc=<0 0 0>, rot=<0 0 0 1>}
 deform_wrist_up_base_l {parent=hand_l, loc=<0 2.3009 0>, rot=<0 0 0 -1>}
 deform_wrist_up_l {parent=deform_wrist_up_base_l, loc=<0 0 0>, rot=<0 0 0 1>}
 lowerarm_twist_02_l {parent=lowerarm_l, loc=<10.6707 0 0>, rot=<0 0 0 1>}
 lowerarm_twist_01_l {parent=lowerarm_l, loc=<18.316 0 0>, rot=<0 0 0 1>}
 deform_elbow_in_base_l {parent=lowerarm_l, loc=<0.1834 3.6062 -0.2572>, rot=<0 0 0 -1>}
 deform_elbow_in_l {parent=deform_elbow_in_base_l, loc=<0 0 0>, rot=<0 0 0 1>}
 deform_elbow_out_base_l {parent=lowerarm_l, loc=<0 -4 0>, rot=<0 0 0 1>}
 deform_elbow_out_l {parent=deform_elbow_out_base_l, loc=<0 0 0>, rot=<0 0 0 1>}
 upperarm_twist_01_l {parent=upperarm_l, loc=<7.2135 -0.1134 0>, rot=<0 0 0 1>}
 deform_upperarm_bk_base_l {parent=upperarm_twist_01_l, loc=<-6.4053 -4.3746 0.314>, rot=<0 0 0 -1>}
 deform_upperarm_bk_l {parent=deform_upperarm_bk_base_l, loc=<0 0 0>, rot=<0 0 0 1>}
 deform_upperarm_fr_base_l {parent=upperarm_twist_01_l, loc=<-5.6312 5.092 0.3354>, rot=<0 0 0 -1>}
 deform_upperarm_fr_l {parent=deform_upperarm_fr_base_l, loc=<0 0 0>, rot=<0 0 0 1>}
 deform_upperarm_up_base_l {parent=upperarm_twist_01_l, loc=<-8.891 0.4825 3.9517>, rot=<0 0 0 -1>}
 deform_upperarm_up_l {parent=deform_upperarm_up_base_l, loc=<0 0 0>, rot=<0 0 0 1>}
 Mstr_scapularMask_base {parent=upperarm_twist_01_l, loc=<0 0 0>, rot=<0 0 0 1>}
 Mstr_scapularMask {parent=Mstr_scapularMask_base, loc=<-12.9869 0.1418 5.8516>, rot=<-0.0075 -0.0246 0.0011 -0.9997>}
 upperarm_twist_02_l {parent=upperarm_l, loc=<19.7998 -0.1134 0>, rot=<0 0 0 1>}
 upperarm_l_DefaultPose {parent=clavicle_l, type=Null, loc=<9.8658 -1.2136 -0.4105>, rot=<0.0164 -0.3245 -0.0269 -0.9454>}
 attach_backpack {parent=spine_05, loc=<0 -8.35 0>, rot=<-0.0093 -0.707 0.0093 -0.707>}
 neck_01 {parent=spine_05, loc=<11.3138 0.1529 0>, rot=<0 0 -0.2003 -0.9797>}
 neck_02 {parent=neck_01, loc=<4.5627 0 0>, rot=<0 0 0.0504 -0.9987>}
 head {parent=neck_02, loc=<5.3914 0 0>, rot=<0 0 0.115 -0.9934>}
 Mstr_BackNeck {parent=neck_01, loc=<2.293 -6.3186 0>, rot=<-0.1137 -0.6979 0.1137 -0.6979>}
 clavicle_r {parent=spine_05, loc=<5.4831 -0.1053 3.2946>, rot=<0.7289 0.0465 -0.6829 0.009>}
 upperarm_r {parent=clavicle_r, loc=<-9.8658 1.2136 0.4104>, rot=<0.0164 -0.3245 -0.0269 -0.9454>}
 lowerarm_r {parent=upperarm_r, loc=<-26.4651 0.8345 -0.259>, rot=<0.0036 0.0161 -0.2298 -0.9731>}
 hand_r {parent=lowerarm_r, loc=<-22.7356 0.3203 0.6106>, rot=<-0.6784 -0.0223 0.0735 -0.7307>}
 thumb_01_r {parent=hand_r, loc=<-1.3119 0.3192 2.459>, rot=<0.6546 -0.4042 0.0664 -0.6354>}
 thumb_02_r {parent=thumb_01_r, loc=<-5.9335 0.0447 -0.0401>, rot=<-0.0037 0.0213 0.298 -0.9543>}
 thumb_03_r {parent=thumb_02_r, loc=<-2.2223 -0.3682 0.0907>, rot=<0.0118 -0.013 0.0066 -0.9998>}
 index_metacarpal_r {parent=hand_r, loc=<-1.6535 -0.0183 2.96>, rot=<0.0158 -0.0124 -0.0173 -0.9996>}
 index_01_r {parent=index_metacarpal_r, loc=<-8.3008 -0.0357 0.0524>, rot=<-0.0044 0.0485 0.0924 -0.9945>}
 index_02_r {parent=index_01_r, loc=<-3.3587 0.3004 -0.0452>, rot=<-0.002 0.0027 0.3225 -0.9466>}
 index_03_r {parent=index_02_r, loc=<-2.5578 0.123 0.0252>, rot=<-0.002 0.0067 0.2558 -0.9667>}
 middle_metacarpal_r {parent=hand_r, loc=<-2.1856 0.1726 0.5984>, rot=<0.006 0.0168 -0.0329 -0.9993>}
 middle_01_r {parent=middle_metacarpal_r, loc=<-7.3275 0.2447 -0.0129>, rot=<-0.0208 0.0042 0.0858 -0.9961>}
 middle_02_r {parent=middle_01_r, loc=<-4.0123 -0.1401 -0.1175>, rot=<-0.0218 0.0258 0.3512 -0.9357>}
 middle_03_r {parent=middle_02_r, loc=<-2.6077 -0.0533 -0.1373>, rot=<0.0082 0.0188 0.1708 -0.9851>}
 ring_metacarpal_r {parent=hand_r, loc=<-2.2073 0.3627 -1.3638>, rot=<-0.0284 0.0722 -0.0458 -0.9959>}
 ring_01_r {parent=ring_metacarpal_r, loc=<-7.0295 0.6197 0.1446>, rot=<-0.0489 -0.0252 0.1892 -0.9804>}
 ring_02_r {parent=ring_01_r, loc=<-3.6174 -0.4647 0.1027>, rot=<0 0 0.2806 -0.9598>}
 ring_03_r {parent=ring_02_r, loc=<-2.4153 -0.1141 -0.0199>, rot=<0 0 0.1189 -0.9929>}
 pinky_metacarpal_r {parent=hand_r, loc=<-1.7617 0.9753 -3.3814>, rot=<-0.0397 0.0814 -0.069 -0.9935>}
 pinky_01_r {parent=pinky_metacarpal_r, loc=<-7.1709 0.4831 0.1094>, rot=<-0.0654 -0.0035 0.2242 -0.9723>}
 pinky_02_r {parent=pinky_01_r, loc=<-2.8274 -0.0892 0.1043>, rot=<0.0037 0.0011 0.2854 -0.9584>}
 pinky_03_r {parent=pinky_02_r, loc=<-2.2046 -0.0307 0.0078>, rot=<-0.0002 0.0098 0.1445 -0.9895>}
 deform_wrist_dn_base_r {parent=hand_r, loc=<-0.0001 1.5328 0>, rot=<0 0 0 -1>}
 deform_wrist_dn_r {parent=deform_wrist_dn_base_r, loc=<0 0 0>, rot=<0 0 0 1>}
 deform_wrist_up_base_r {parent=hand_r, loc=<-0.0002 -2.3007 -0.0001>, rot=<0 0 0 -1>}
 deform_wrist_up_r {parent=deform_wrist_up_base_r, loc=<0 0 0>, rot=<0 0 0 1>}
 lowerarm_twist_02_r {parent=lowerarm_r, loc=<-10.6707 0 0>, rot=<0 0 0 1>}
 lowerarm_twist_01_r {parent=lowerarm_r, loc=<-18.3161 0 0>, rot=<0 0 0 1>}
 deform_elbow_in_base_r {parent=lowerarm_r, loc=<-0.1834 -3.6063 0.2572>, rot=<0 0 0 -1>}
 deform_elbow_in_r {parent=deform_elbow_in_base_r, loc=<0 0 0>, rot=<0 0 0 1>}
 deform_elbow_out_base_r {parent=lowerarm_r, loc=<0 4 -0.0001>, rot=<0 0 0 1>}
 deform_elbow_out_r {parent=deform_elbow_out_base_r, loc=<0 0 0>, rot=<0 0 0 1>}
 upperarm_twist_01_r {parent=upperarm_r, loc=<-7.2135 0.1134 0>, rot=<0 0 0 1>}
 deform_upperarm_bk_base_r {parent=upperarm_twist_01_r, loc=<6.4052 4.3746 -0.314>, rot=<0 0 0 -1>}
 deform_upperarm_bk_r {parent=deform_upperarm_bk_base_r, loc=<0 0 0>, rot=<0 0 0 1>}
 deform_upperarm_fr_base_r {parent=upperarm_twist_01_r, loc=<5.6311 -5.092 -0.3354>, rot=<0 0 0 -1>}
 deform_upperarm_fr_r {parent=deform_upperarm_fr_base_r, loc=<0 0 0>, rot=<0 0 0 1>}
 deform_upperarm_up_base_r {parent=upperarm_twist_01_r, loc=<8.891 -0.4825 -3.9517>, rot=<0 0 0 -1>}
 deform_upperarm_up_r {parent=deform_upperarm_up_base_r, loc=<0 0 0>, rot=<0 0 0 1>}
 upperarm_twist_02_r {parent=upperarm_r, loc=<-19.7998 0.1134 0>, rot=<0 0 0 1>}
 upperarm_r_DefaultPose {parent=clavicle_r, type=Null, loc=<-9.8658 1.2136 0.4104>, rot=<0.0164 -0.3245 -0.0269 -0.9454>}
 Mstr_edge_r {parent=clavicle_r, loc=<-9.719 -4.9532 2.6569>, rot=<0.7138 -0.6996 0.024 -0.0235>}
 Dyn_collar_l {parent=spine_05, loc=<7.7816 7.3797 -3.4955>, rot=<0.0701 -0.3067 0.9216 -0.2273>}
 Dyn_collar_r {parent=spine_05, loc=<7.7818 7.3798 3.4955>, rot=<0.3067 0.0701 -0.2273 -0.9216>}
 Dyn_Necklace {parent=spine_05, loc=<11.3138 0.1529 0>, rot=<0.2615 0.657 0.657 0.2615>}
 Dyn_Necklace_gear {parent=Dyn_Necklace, loc=<11.4207 0 0>, rot=<0 0 0 1>}
 Dyn_Amulete {parent=Dyn_Necklace_gear, loc=<1.4988 0 0.1222>, rot=<0 -0.1327 0 -0.9912>}
 Dyn_Amulete_end {parent=Dyn_Amulete, loc=<3.1009 0 0>, rot=<0 0 0 1>}
 clavicle_l_DefaultPose {parent=spine_05, type=Null, loc=<5.4831 -0.1053 -3.2946>, rot=<0.0465 -0.7289 -0.009 -0.6829>}
 deform_pec_aim_base_l {parent=spine_04, loc=<3.2456 9.3309 -1.2237>, rot=<0.0603 -0.4033 0.3318 -0.8507>}
 deform_pec_aim_l {parent=deform_pec_aim_base_l, loc=<0 0 0>, rot=<0 0 0 1>}
 deform_pec_base_l {parent=deform_pec_aim_l, loc=<9.3665 0 0>, rot=<0 0 0 1>}
 deform_pec_l {parent=deform_pec_base_l, loc=<0 0 0>, rot=<0 0 0 1>}
 deform_pec_aim_base_r {parent=spine_04, loc=<3.2452 9.3309 1.2237>, rot=<-0.4033 -0.0603 0.8507 0.3318>}
 deform_pec_aim_r {parent=deform_pec_aim_base_r, loc=<0 0 0>, rot=<0 0 0 -1>}
 deform_pec_base_r {parent=deform_pec_aim_r, loc=<-9.3667 -0.0003 -0.0004>, rot=<0 0 0 -1>}
 deform_pec_r {parent=deform_pec_base_r, loc=<0 0 0>, rot=<0 0 0 -1>}
 Mstr_Jacket_front_r {parent=spine_04, loc=<1.6745 13.4735 4.4095>, rot=<0.4866 0.513 0.5276 0.4708>}
 Dyn_Jacket_front_r {parent=Mstr_Jacket_front_r, loc=<0 0 0>, rot=<0.5 -0.5 -0.5 -0.5>}
 Dyn_Zip {parent=Dyn_Jacket_front_r, loc=<7.3716 0.0119 -0.5189>, rot=<-0.7071 0.0079 0.7071 -0.0079>}
 Mstr_Jacket_Back_r {parent=spine_04, loc=<2.9052 -8.6141 7.773>, rot=<0.6085 -0.3602 0.5765 -0.4095>}
 Dyn_Jacket_Back_r {parent=Mstr_Jacket_Back_r, loc=<0 0 0>, rot=<0.5 -0.5 -0.5 -0.5>}
 Mstr_Jacket_Back_c {parent=spine_04, loc=<3.076 -9.2311 0>, rot=<0.5204 -0.4788 0.4788 -0.5204>}
 Dyn_Jacket_Back_c {parent=Mstr_Jacket_Back_c, loc=<0 0 0>, rot=<0.5 -0.5 -0.5 -0.5>}
 Mstr_Jacket_Back_l {parent=spine_04, loc=<2.9055 -8.614 -7.773>, rot=<0.3602 0.6085 -0.4095 -0.5765>}
 Dyn_Jacket_Back_l {parent=Mstr_Jacket_Back_l, loc=<0 0 0>, rot=<0.5 -0.5 -0.5 -0.5>}
 Mstr_Jacket_front_l {parent=spine_04, loc=<1.6745 13.4735 -4.4095>, rot=<-0.513 0.4866 0.4708 -0.5276>}
 Dyn_Jacket_front_l {parent=Mstr_Jacket_front_l, loc=<0 0 0>, rot=<0.5 -0.5 -0.5 -0.5>}
 Mstr_Jacket_armpit_r {parent=spine_04, loc=<3.436 -2.2846 12.5284>, rot=<0.6085 -0.3602 0.5765 -0.4095>}
 Mstr_Jacket_armpit_l {parent=spine_04, loc=<3.4363 -2.2845 -12.5284>, rot=<0.3602 0.6085 -0.4095 -0.5765>}
 thigh_l {parent=pelvis, loc=<-1.0855 -0.2109 -8.7053>, rot=<0.0009 0.0355 0.0155 -0.9993>}
 calf_l {parent=thigh_l, loc=<-42.5284 -0.0019 -0.7297>, rot=<0.0006 -0.0161 -0.0391 -0.9991>}
 foot_l {parent=calf_l, loc=<-46.7402 -2.5014 -0.592>, rot=<0.0427 -0.0238 0.0353 -0.9982>}
 ball_l {parent=foot_l, loc=<-3.5658 10.8066 -0.1544>, rot=<0.0001 -0.0001 -0.7063 -0.7079>}
 calf_twist_02_l {parent=calf_l, loc=<-19.1546 -1.7943 -0.2638>, rot=<0.0237 0.0095 -0.0035 -0.9997>}
 calf_twist_01_l {parent=calf_l, loc=<-35.8483 -1.9197 -0.5794>, rot=<0.0237 0.0095 -0.0035 -0.9997>}
 Mstr_pants_cuffs_outter_l {parent=calf_twist_01_l, loc=<-1.6444 3.5452 -2.0759>, rot=<0.0068 0.832 -0.0303 -0.5539>}
 Dyn_pants_cuffs_outter_l {parent=Mstr_pants_cuffs_outter_l, loc=<0 0 0>, rot=<0.5 -0.5 -0.5 -0.5>}
 Mstr_pants_cuffs_back_l {parent=calf_twist_01_l, loc=<-0.9906 -6.0544 0.3024>, rot=<-0.4517 -0.5209 0.5533 0.4675>}
 Dyn_pants_cuffs_back_l {parent=Mstr_pants_cuffs_back_l, loc=<0 0 0>, rot=<0.5 -0.5 -0.5 -0.5>}
 Mstr_pants_cuffs_Inner_l {parent=calf_twist_01_l, loc=<-1.32 3.5262 1.2634>, rot=<-0.5169 0.0798 0.8491 -0.0743>}
 Dyn_pants_cuffs_Inner_l {parent=Mstr_pants_cuffs_Inner_l, loc=<0 0 0>, rot=<0.5 -0.5 -0.5 -0.5>}
 deform_knee_in_base_l {parent=calf_l, loc=<0.8031 -5.0231 0.2456>, rot=<0 0 0 -1>}
 deform_knee_in_l {parent=deform_knee_in_base_l, loc=<0 0 0>, rot=<0 0 0 1>}
 deform_knee_out_base_l {parent=calf_l, loc=<0 4 0>, rot=<0 0 0 -1>}
 deform_knee_out_l {parent=deform_knee_out_base_l, loc=<0 0 0>, rot=<0 0 0 1>}
 foot_l_DefaultPos {parent=calf_l, type=Null, loc=<-46.7402 -2.5014 -0.592>, rot=<0.0427 -0.0238 0.0353 -0.9982>}
 thigh_twist_01_l {parent=thigh_l, loc=<-27.952 0 0>, rot=<0.0013 0 0 -1>}
 Mstr_wrench {parent=thigh_twist_01_l, loc=<15.1568 -0.1231 -10.741>, rot=<0.0456 0.6677 0.7418 -0.042>}
 Dyn_wrench {parent=Mstr_wrench, loc=<0 0 0>, rot=<0 0 0 1>}
 Dyn_wrench_end {parent=Dyn_wrench, loc=<13.8568 0 0>, rot=<0 0 0 1>}
 Mstr_screwdriver {parent=thigh_twist_01_l, loc=<15.8847 7.4529 -7.173>, rot=<0.048 -0.9339 -0.353 0.0307>}
 Dyn_screwdriver {parent=Mstr_screwdriver, loc=<0 0 0>, rot=<0 0 0 1>}
 Dyn_screwdriver_end {parent=Dyn_screwdriver, loc=<9.3204 0 0>, rot=<0 0 0 1>}
 Mstr_belt_left {parent=thigh_twist_01_l, loc=<19.294 5.8805 -3.4239>, rot=<0.0237 -0.86 0.0209 -0.5093>}
 Mstr_belt_back {parent=thigh_twist_01_l, loc=<19.1905 -5.9296 -6.3972>, rot=<0.7525 -0.2339 0.531 -0.3115>}
 deform_groin_base_l {parent=thigh_l, loc=<0.4801 5.6776 -0.2486>, rot=<0 0 0 -1>}
 deform_groin_l {parent=deform_groin_base_l, loc=<0 0 0>, rot=<0 0 0 1>}
 thigh_r {parent=pelvis, loc=<-1.0855 -0.2109 8.7053>, rot=<0.0355 -0.0009 0.9993 0.0155>}
 calf_r {parent=thigh_r, loc=<42.5284 0.0019 0.7297>, rot=<0.0006 -0.0161 -0.0391 -0.9991>}
 foot_r {parent=calf_r, loc=<46.7401 2.5014 0.592>, rot=<0.0427 -0.0238 0.0353 -0.9982>}
 ball_r {parent=foot_r, loc=<3.5658 -10.8066 0.1544>, rot=<0.0001 -0.0001 -0.7063 -0.7079>}
 calf_twist_02_r {parent=calf_r, loc=<19.1546 1.7943 0.2638>, rot=<0.0237 0.0095 -0.0035 -0.9997>}
 calf_twist_01_r {parent=calf_r, loc=<35.8483 1.9197 0.5794>, rot=<0.0237 0.0095 -0.0035 -0.9997>}
 Mstr_pants_cuffs_Inner_r {parent=calf_twist_01_r, loc=<1.3199 -3.5262 -1.2634>, rot=<-0.5169 0.0798 0.8491 -0.0743>}
 Dyn_pants_cuffs_Inner_r {parent=Mstr_pants_cuffs_Inner_r, loc=<0 0 0>, rot=<0.5 -0.5 -0.5 -0.5>}
 Mstr_pants_cuffs_outter_r {parent=calf_twist_01_r, loc=<1.6443 -3.5452 2.0759>, rot=<0.0068 0.832 -0.0303 -0.5539>}
 Dyn_pants_cuffs_outter_r {parent=Mstr_pants_cuffs_outter_r, loc=<0 0 0>, rot=<0.5 -0.5 -0.5 -0.5>}
 Mstr_pants_cuffs_back_r {parent=calf_twist_01_r, loc=<0.9904 6.0544 -0.3024>, rot=<-0.4517 -0.5209 0.5533 0.4675>}
 Dyn_pants_cuffs_back_r {parent=Mstr_pants_cuffs_back_r, loc=<0 0 0>, rot=<0.5 -0.5 -0.5 -0.5>}
 deform_knee_in_base_r {parent=calf_r, loc=<-0.8031 5.0231 -0.2457>, rot=<0 0 0 -1>}
 deform_knee_in_r {parent=deform_knee_in_base_r, loc=<0 0 0>, rot=<0 0 0 1>}
 deform_knee_out_base_r {parent=calf_r, loc=<0 -4 0>, rot=<0 0 0 -1>}
 deform_knee_out_r {parent=deform_knee_out_base_r, loc=<0 0 0>, rot=<0 0 0 1>}
 foot_r_DefaultPos {parent=calf_r, type=Null, loc=<46.7403 2.5014 0.592>, rot=<0.0427 -0.0238 0.0353 -0.9982>}
 thigh_twist_01_r {parent=thigh_r, loc=<27.952 0 0>, rot=<0.0013 0 0 -1>}
 deform_groin_base_r {parent=thigh_r, loc=<-0.48 -5.6776 0.2486>, rot=<0 0 0 -1>}
 deform_groin_r {parent=deform_groin_base_r, loc=<0 0 0>, rot=<0 0 0 1>}
 deform_glute_base_l {parent=pelvis, loc=<-7.3466 -9.0213 -9.1245>, rot=<0.0172 0.031 -0.4477 -0.8935>}
 deform_glute_l {parent=deform_glute_base_l, loc=<0 0 0>, rot=<0 0 0 1>}
 deform_glute_base_r {parent=pelvis, loc=<-7.3464 -9.0213 9.1245>, rot=<0.031 -0.0172 0.8935 -0.4477>}
 deform_glute_r {parent=deform_glute_base_r, loc=<0 0 0>, rot=<0 0 0 1>}
 Mstr_belt_mid {parent=pelvis, loc=<-3.4648 7.435 -0.2105>, rot=<0.0107 -0.8416 0.0315 -0.5391>}
 Mstr_belt_right {parent=pelvis, loc=<2.0667 4.9315 11.616>, rot=<-0.1566 -0.8307 -0.0743 -0.529>}
 thigh_r_DefaultPos {parent=pelvis, type=Null, loc=<-1.0853 -0.2109 8.7053>, rot=<0.0355 -0.0009 0.9993 0.0155>}
 thigh_l_DefaultPos {parent=pelvis, type=Null, loc=<-1.0855 -0.2109 -8.7053>, rot=<0.0009 0.0355 0.0155 -0.9993>}
 Mstr_belt_back_glute {parent=pelvis, loc=<-5.2447 -12.2465 -8.3136>, rot=<0.7943 -0.0105 0.5915 -0.1383>}
]
variables [
 "Calfs Bend  Alpha" {type=double, default=0}
 "Debug Drow " {
  type=FSphericalPoseReaderDebugSettings
  default=(bDrawDebug=False,bDraw2D=False,bDrawLocalAxes=False,DebugScale=13.0012,DebugSegments=20,DebugThickness=0.5397)
 }
 FrontEnd Idle Bypass {type=bool, default=False}
 skyDive {type=bool, default=False}
]
