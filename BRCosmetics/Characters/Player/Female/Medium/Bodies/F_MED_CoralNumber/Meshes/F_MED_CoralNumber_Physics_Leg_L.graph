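# Physics asset for the left leg of the F_MED_CoralNumber female body cosmetic, set up for secondary motion of the pants cuff.
# Inner, back and outer cuff chains each link a small Mstr sphere to calf_l, then a larger Dyn box or capsule to that sphere.
# All joints lock twist and set narrow swing limits, with the back Dyn joint also locking swing2.
# Calf and foot capsules supply leg shapes, and the foot has no constraint.
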PhysicsAsset /BRCosmetics/Characters/Player/Female/Medium/Bodies/F_MED_CoralNumber/Meshes/F_MED_CoralNumber_Physics_Leg_L
bodies {
 Dyn_pants_cuffs_Inner_l [{
   shapeType=Box
   extentX=12.1484
   extentY=3.0678
   extentZ=7.5
   shapeName=Dyn_pants_cuffs_back_r_box
   center=<-11.5674 0.0114 0.0198>
   rotation=<rot 0.8337 4.1201 -11.418>
  }]
 Dyn_pants_cuffs_back_l [{shapeType=Box, extentX=12.3477, extentY=2.7477, extentZ=7.5, shapeName=Dyn_pants_cuffs_back_l_box, center=<-9.2768 0 0>}]
 Dyn_pants_cuffs_outter_l [{
   shapeType=Capsule
   length=6.1583
   radius=1.4815
   shapeName=Dyn_Jacket_front_r_capsule
   center=<-10.1885 0 0>
   rotation=<rot -90 179.7468 0.253>
  }]
 Mstr_pants_cuffs_Inner_l [{shapeType=Sphere, radius=0.9745, shapeName=Mstr_Jacket_front_r_sphere}]
 Mstr_pants_cuffs_back_l [{shapeType=Sphere, radius=0.9745, shapeName=Mstr_Jacket_front_r_sphere}]
 Mstr_pants_cuffs_outter_l [{shapeType=Sphere, radius=0.9745, shapeName=Mstr_Jacket_front_r_sphere}]
 calf_l [{
   shapeType=Capsule
   length=8.0538
   radius=3.4237
   shapeName=calf_r_capsule
   center=<-44.3295 -2.0736 -0.6587>
   rotation=<rot -90 179.9999 -0.0001>
  }]
 foot_l [{shapeType=Capsule, length=8.83, radius=5.0723, shapeName=foot_r_capsule, center=<-2.9978 3.6084 0>, rotation=<rot 0 -0.0002 90>}]
}
constraints [
 {bone1Name=Mstr_pants_cuffs_Inner_l, bone2Name=calf_l, swing1LimitDegrees=30, swing2LimitDegrees=15, twistLimitDegrees=45, twistMotion=Locked}
 {
  bone1Name=Dyn_pants_cuffs_Inner_l
  bone2Name=Mstr_pants_cuffs_Inner_l
  swing1LimitDegrees=15
  swing2LimitDegrees=10
  twistLimitDegrees=45
  twistMotion=Locked
 }
 {bone1Name=Mstr_pants_cuffs_back_l, bone2Name=calf_l, swing1LimitDegrees=30, swing2LimitDegrees=10, twistLimitDegrees=45, twistMotion=Locked}
 {
  bone1Name=Dyn_pants_cuffs_back_l
  bone2Name=Mstr_pants_cuffs_back_l
  swing1LimitDegrees=15
  swing2LimitDegrees=15
  swing2Motion=Locked
  twistLimitDegrees=45
  twistMotion=Locked
 }
 {
  bone1Name=Mstr_pants_cuffs_outter_l
  bone2Name=calf_l
  swing1LimitDegrees=30
  swing2LimitDegrees=15
  twistLimitDegrees=45
  twistMotion=Locked
 }
 {
  bone1Name=Dyn_pants_cuffs_outter_l
  bone2Name=Mstr_pants_cuffs_outter_l
  swing1LimitDegrees=15
  swing2LimitDegrees=15
  twistLimitDegrees=45
  twistMotion=Locked
 }
]
