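# Physics asset for the right leg of the F_MED_CoralNumber female character body, simulating the pants cuffs.
# Back, inner and outer Dyn cuff bodies, using boxes and a capsule, are each constrained to a small Mstr sphere body.
# Constraints lock twist and limit swing to narrow angles, with the back cuff also locking its second swing axis.
# Capsules on calf_r and foot_r add leg shapes alongside the cuffs.
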PhysicsAsset /BRCosmetics/Characters/Player/Female/Medium/Bodies/F_MED_CoralNumber/Meshes/F_MED_CoralNumber_Physics_Leg_R
bodies {
 Dyn_pants_cuffs_Inner_r [{
   shapeType=Box
   extentX=12.1484
   extentY=3.0678
   extentZ=7.5
   shapeName=Dyn_pants_cuffs_back_r_box
   center=<9.9756 -0.3604 0>
   rotation=<rot 0.0001 0 -4.4869>
  }]
 Dyn_pants_cuffs_back_r [{shapeType=Box, extentX=12.1484, extentY=3.0678, extentZ=7.5, shapeName=Dyn_pants_cuffs_back_r_box, center=<9.9756 0 0>}]
 Dyn_pants_cuffs_outter_r [{
   shapeType=Capsule
   length=6.1583
   radius=1.4815
   shapeName=Dyn_Jacket_front_r_capsule
   center=<10.1885 0 0>
   rotation=<rot -90 177.4696 -177.4698>
  }]
 Mstr_pants_cuffs_Inner_r [{shapeType=Sphere, radius=0.9745, shapeName=Mstr_Jacket_front_r_sphere}]
 Mstr_pants_cuffs_back_r [{shapeType=Sphere, radius=0.9745, shapeName=Mstr_Jacket_front_r_sphere}]
 Mstr_pants_cuffs_outter_r [{shapeType=Sphere, radius=0.9745, shapeName=Mstr_Jacket_front_r_sphere}]
 calf_r [{
   shapeType=Capsule
   length=8.0538
   radius=3.4237
   shapeName=calf_r_capsule
   center=<44.3295 2.0736 0.6587>
   rotation=<rot -90 179.9999 179.9999>
  }]
 foot_r [{shapeType=Capsule, length=8.83, radius=5.0723, shapeName=foot_r_capsule, center=<2.9978 -3.6084 0>, rotation=<rot 0 -0.0002 -90>}]
}
constraints [
 {
  bone1Name=Dyn_pants_cuffs_back_r
  bone2Name=Mstr_pants_cuffs_back_r
  swing1LimitDegrees=15
  swing2LimitDegrees=10
  swing2Motion=Locked
  twistLimitDegrees=45
  twistMotion=Locked
 }
 {
  bone1Name=Dyn_pants_cuffs_outter_r
  bone2Name=Mstr_pants_cuffs_outter_r
  swing1LimitDegrees=15
  swing2LimitDegrees=15
  twistLimitDegrees=45
  twistMotion=Locked
 }
 {
  bone1Name=Dyn_pants_cuffs_Inner_r
  bone2Name=Mstr_pants_cuffs_Inner_r
  swing1LimitDegrees=15
  swing2LimitDegrees=10
  twistLimitDegrees=45
  twistMotion=Locked
 }
]
