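# Physics asset for the wrench and screwdriver accessories on the F_MED_CoralNumber female body cosmetic, apparently for secondary dangle motion.
# Each tool has a capsule body on its Dyn bone, constrained to a small sphere body on its matching Mstr bone.
# Both constraints lock twist, and the screwdriver also locks its second swing axis.
# Shape names still refer to the right jacket front, which suggests the bodies were copied from that setup.
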PhysicsAsset /BRCosmetics/Characters/Player/Female/Medium/Bodies/F_MED_CoralNumber/Meshes/F_MED_CoralNumber_Physics_Wrench
bodies {
 Dyn_screwdriver [{
   shapeType=Capsule
   length=13.9795
   radius=1.2936
   shapeName=Dyn_Jacket_front_r_capsule
   center=<1.9287 0 0>
   rotation=<rot -90 166.7373 -166.7376>
  }]
 Dyn_wrench [{
   shapeType=Capsule
   length=24.1513
   radius=1.2936
   shapeName=Dyn_Jacket_front_r_capsule
   center=<3.4578 0 0>
   rotation=<rot -90 166.7373 -166.7376>
  }]
 Mstr_screwdriver [{shapeType=Sphere, radius=0.9745, shapeName=Mstr_Jacket_front_r_sphere}]
 Mstr_wrench [{shapeType=Sphere, radius=0.9745, shapeName=Mstr_Jacket_front_r_sphere}]
}
constraints [
 {
  bone1Name=Dyn_screwdriver
  bone2Name=Mstr_screwdriver
  swing1LimitDegrees=10
  swing2LimitDegrees=5
  swing2Motion=Locked
  twistLimitDegrees=45
  twistMotion=Locked
 }
 {bone1Name=Dyn_wrench, bone2Name=Mstr_wrench, swing1LimitDegrees=2, swing2LimitDegrees=4, twistLimitDegrees=45, twistMotion=Locked}
]
