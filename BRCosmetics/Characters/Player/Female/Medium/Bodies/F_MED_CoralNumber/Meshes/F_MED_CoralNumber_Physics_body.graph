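# Physics asset for the CoralNumber medium female player body cosmetic, set up for secondary motion on jacket, collar, zip and necklace.
# Capsules on Dyn_ jacket panels are constrained to small Mstr_ anchor spheres, with tighter limits at the back than the front.
# Collars attach to spine_05, as does a necklace-gear-amulet chain, and the zip hangs off the front-right panel.
# Locked axes restrict zip and necklace motion; an unconstrained spine_04 box covers the upper torso.
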PhysicsAsset /BRCosmetics/Characters/Player/Female/Medium/Bodies/F_MED_CoralNumber/Meshes/F_MED_CoralNumber_Physics_body
bodies {
 Dyn_Amulete [{
   shapeType=Capsule
   length=3.6955
   radius=0.9522
   shapeName=Dyn_Jacket_front_r_capsule
   center=<3.2426 0 0.4973>
   rotation=<rot -90 165.9638 104.036>
  }]
 Dyn_Jacket_Back_c [{
   shapeType=Capsule
   length=3.8508
   radius=1.2936
   shapeName=Dyn_Jacket_front_r_capsule
   center=<5.9147 0 0>
   rotation=<rot -90 166.7373 -166.7376>
  }]
 Dyn_Jacket_Back_l [{
   shapeType=Capsule
   length=3.8508
   radius=1.2936
   shapeName=Dyn_Jacket_front_r_capsule
   center=<-6.0641 0 0>
   rotation=<rot -90 177.4696 -177.4698>
  }]
 Dyn_Jacket_Back_r [{
   shapeType=Capsule
   length=3.8508
   radius=1.2936
   shapeName=Dyn_Jacket_front_r_capsule
   center=<6.1265 0 0>
   rotation=<rot -90 6.7214 -6.7216>
  }]
 Dyn_Jacket_front_l [{
   shapeType=Capsule
   length=3.8508
   radius=1.2936
   shapeName=Dyn_Jacket_front_r_capsule
   center=<-3.4578 0 0>
   rotation=<rot -90 166.7373 -166.7376>
  }]
 Dyn_Jacket_front_r [{
   shapeType=Capsule
   length=3.8508
   radius=1.2936
   shapeName=Dyn_Jacket_front_r_capsule
   center=<3.4578 0 0>
   rotation=<rot -90 166.7373 -166.7376>
  }]
 Dyn_Necklace [{shapeType=Sphere, radius=1.1518, shapeName=Dyn_Necklace_sphere, center=<7.9261 0 0>}]
 Dyn_Necklace_gear [{
   shapeType=Capsule
   length=3.9219
   radius=0.7258
   shapeName=Dyn_Jacket_front_r_capsule
   center=<2.58 0 0.2218>
   rotation=<rot -84.3488 180 179.9997>
  }]
 Dyn_Zip [{shapeType=Capsule, length=2.7331, radius=0.7517, shapeName=Dyn_Jacket_front_r_capsule, center=<0 0 -1.8728>}]
 Dyn_collar_l [{
   shapeType=Capsule
   length=1.9809
   radius=1.2936
   shapeName=Dyn_Jacket_front_r_capsule
   center=<-1.8909 0 0>
   rotation=<rot -90 6.7214 -6.7216>
  }]
 Dyn_collar_r [{
   shapeType=Capsule
   length=1.9809
   radius=1.2936
   shapeName=Dyn_Jacket_front_r_capsule
   center=<1.8909 0 0>
   rotation=<rot -90 6.7214 -6.7216>
  }]
 Mstr_Jacket_Back_c [{shapeType=Sphere, radius=0.9745, shapeName=Mstr_Jacket_front_r_sphere}]
 Mstr_Jacket_Back_l [{shapeType=Sphere, radius=0.9745, shapeName=Mstr_Jacket_front_r_sphere}]
 Mstr_Jacket_Back_r [{shapeType=Sphere, radius=0.9745, shapeName=Mstr_Jacket_front_r_sphere}]
 Mstr_Jacket_front_l [{shapeType=Sphere, radius=0.9745, shapeName=Mstr_Jacket_front_r_sphere}]
 Mstr_Jacket_front_r [{shapeType=Sphere, radius=0.9745, shapeName=Mstr_Jacket_front_r_sphere}]
 spine_04 [{shapeType=Box, extentX=9.4558, extentY=7.5, extentZ=9.9168, shapeName=spine_05_box, center=<5.3023 6.8135 0>, rotation=<rot 0 -33.303 0>}]
 spine_05 [{shapeType=Sphere, radius=1.366, shapeName=spine_05_sphere}]
}
constraints [
 {
  bone1Name=Dyn_Zip
  bone2Name=Dyn_Jacket_front_r
  swing1LimitDegrees=45
  swing1Motion=Locked
  swing2LimitDegrees=45
  swing2Motion=Locked
  twistLimitDegrees=60
 }
 {bone1Name=Dyn_Jacket_front_r, bone2Name=Mstr_Jacket_front_r, swing1LimitDegrees=20, swing2LimitDegrees=20, twistLimitDegrees=10}
 {bone1Name=Dyn_Jacket_front_l, bone2Name=Mstr_Jacket_front_l, swing1LimitDegrees=20, swing2LimitDegrees=20, twistLimitDegrees=10}
 {bone1Name=Dyn_Jacket_Back_r, bone2Name=Mstr_Jacket_Back_r, swing1LimitDegrees=10, swing2LimitDegrees=5, twistLimitDegrees=5}
 {bone1Name=Dyn_Jacket_Back_c, bone2Name=Mstr_Jacket_Back_c, swing1LimitDegrees=10, swing2LimitDegrees=5, twistLimitDegrees=5}
 {bone1Name=Dyn_Jacket_Back_l, bone2Name=Mstr_Jacket_Back_l, swing1LimitDegrees=10, swing2LimitDegrees=5, twistLimitDegrees=5}
 {bone1Name=Dyn_collar_l, bone2Name=spine_05, swing1LimitDegrees=5, swing2LimitDegrees=45, twistLimitDegrees=10}
 {bone1Name=Dyn_collar_r, bone2Name=spine_05, swing1LimitDegrees=5, swing2LimitDegrees=45, twistLimitDegrees=10}
 {bone1Name=Dyn_Amulete, bone2Name=Dyn_Necklace_gear, swing1LimitDegrees=5, swing2LimitDegrees=2, twistLimitDegrees=45, twistMotion=Locked}
 {
  bone1Name=Dyn_Necklace_gear
  bone2Name=Dyn_Necklace
  swing1LimitDegrees=45
  swing1Motion=Locked
  swing2LimitDegrees=2
  twistLimitDegrees=45
  twistMotion=Locked
 }
 {
  bone1Name=spine_05
  bone2Name=Dyn_Necklace
  swing1LimitDegrees=20
  swing1Motion=Locked
  swing2LimitDegrees=45
  swing2Motion=Locked
  twistLimitDegrees=1
 }
]
