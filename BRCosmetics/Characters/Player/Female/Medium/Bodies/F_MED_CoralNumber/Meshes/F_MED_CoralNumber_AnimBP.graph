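# Post-process style Animation Blueprint for the F_MED_CoralNumber female body cosmetic, adding secondary motion on top of the parent mesh's pose.
# The copied pose runs through the Control Rig, then chained base-bone-space RigidBody sims for body, left leg, right leg and wrench.
# The event graph lowers leg sim alpha while skydiving, flags front-end non-emote idle for the rig, and computes a gravity override via a deprecated helper.
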AnimBlueprint /BRCosmetics/Characters/Player/Female/Medium/Bodies/F_MED_CoralNumber/Meshes/F_MED_CoralNumber_AnimBP
target_skeleton=&/F_MED_CoralNumber_Skeleton
graphs [
 AnimGraph {nodes [
   Input Pose()
   n1 = Control Rig!(
    Source=Copy Pose From Mesh() {bUseAttachedParent=true}
    Alpha
    skyDive:bool=Get bIsSkyDiving()
    FrontEnd Idle Bypass:bool=Get FrontEnd Idle Bypass() -> (FrontEnd Idle Bypass:bool)
   ) -> (Pose) {
    alphaBoolBlend {blendOption=Linear}
    controlRigAssetReference {blueprintRigClass=&/F_MED_CoralNumber_CtrlRig.F_MED_CoralNumber_CtrlRig_C}
   }
   n2 = RigidBody!(ComponentPose=Local To Component(LocalPose=@n1), Alpha, OverrideWorldGravity:Vector="0,0,0", ExternalForce) -> (Pose) {
    bOverrideWorldGravity=true
    overridePhysicsAsset=&/F_MED_CoralNumber_Physics_body
    simulationSpace=BaseBoneSpace
    alphaBoolBlend {blendOption=Linear}
    baseBoneRef {boneName=spine_03}
    componentAppliedLinearAccClamp=<1800 1800 1800>
    componentLinearAccScale=<0.25 0.25 0.25>
    componentLinearVelScale=<0.25 0.25 0.25>
    simSpaceSettings {maxAngularAcceleration=50, maxAngularVelocity=10, maxLinearAcceleration=50, maxLinearVelocity=10, velocityScaleZ=0, worldAlpha=0.2}
   }
   n3 = Get Alpha() -> (Alpha:float)
   n4 = RigidBody(ComponentPose=@n2, Alpha=@n3, OverrideWorldGravity="0,0,0") {
    bOverrideWorldGravity=true
    overridePhysicsAsset=&/F_MED_CoralNumber_Physics_Leg_L
    simulationSpace=BaseBoneSpace
    alphaBoolBlend {blendOption=Linear}
    baseBoneRef {boneName=calf_l}
    componentAppliedLinearAccClamp=<700 700 700>
    componentLinearAccScale=<1 1 1>
    componentLinearVelScale=<1 1 1>
    simSpaceSettings {maxAngularAcceleration=100, maxAngularVelocity=20, maxLinearAcceleration=100, maxLinearVelocity=20, velocityScaleZ=0}
   }
   n5 = RigidBody(ComponentPose=@n4, Alpha=@n3, OverrideWorldGravity="0,0,0") {
    bOverrideWorldGravity=true
    overridePhysicsAsset=&/F_MED_CoralNumber_Physics_Leg_R
    simulationSpace=BaseBoneSpace
    alphaBoolBlend {blendOption=Linear}
    baseBoneRef {boneName=calf_r}
    componentAppliedLinearAccClamp=<700 700 700>
    componentLinearAccScale=<1 1 1>
    componentLinearVelScale=<1 1 1>
    simSpaceSettings {maxAngularAcceleration=100, maxAngularVelocity=20, maxLinearAcceleration=100, maxLinearVelocity=20, velocityScaleZ=0}
   }
   n6 = RigidBody(ComponentPose=@n5, OverrideWorldGravity="0,0,0") {
    bOverrideWorldGravity=true
    overridePhysicsAsset=&/F_MED_CoralNumber_Physics_Wrench
    simulationSpace=BaseBoneSpace
    alphaBoolBlend {blendOption=Linear}
    baseBoneRef {boneName=thigh_l}
    componentAppliedLinearAccClamp=<1000 1000 1000>
    componentLinearAccScale=<1 1 1>
    componentLinearVelScale=<1 1 1>
    simSpaceSettings {maxAngularAcceleration=100, maxAngularVelocity=20, maxLinearAcceleration=100, maxLinearVelocity=20}
   }
   Output Pose(Result=Component To Local(ComponentPose=@n6))
  ]}
 EventGraph {nodes [
   Event BlueprintUpdateAnimation()
   > n1 = Sequence()
   > n2 = Gravity Override State Function DEPRECATED(CharacterPartAnimInstance=Self-Reference(), GameTimeInSeconds=GetGameTimeInSeconds(), GravityOverrideParams=Get Gravity Override Params())
   > Set Gravity Override(Gravity Override=@n2)
   > Set Alpha(Alpha:float=SelectFloat(A=0.7, B=1, bPickA=Get bIsSkyDiving())) -> (Output_Get:float)
   Branch(execute=@n1.then_1, Condition=Get bIsInFrontEnd())
   > Set FrontEnd Idle Bypass(FrontEnd Idle Bypass:bool=NOT Boolean(A=Get bIsPlayingEmote())) -> (Output_Get:bool)
   Event OnCharacterCustomizationLoaded()
  ]}
]
variables [
 Alpha {type=number, default=1}
 FrontEnd Idle Bypass {type=boolean}
 Gravity Override {type=Vector}
 Gravity Override Params {type=GravityOverrideParamsStruct}
]
