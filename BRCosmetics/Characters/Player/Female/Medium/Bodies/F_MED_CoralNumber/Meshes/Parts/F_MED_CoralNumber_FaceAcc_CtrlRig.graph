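# Control Rig for the F_MED_CoralNumber face accessory mesh that applies corrective offsets to the back hair and ponytail as the head rotates.
# It measures the head bone relative to a head_DefaultPose null and converts the rotation to Euler Z.
# Two remaps of that angle weight chained ModifyTransforms nodes in successive negative-angle ranges, layering a second offset on the first at more extreme angles.
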ControlRigBlueprint /BRCosmetics/Characters/Player/Female/Medium/Bodies/F_MED_CoralNumber/Meshes/Parts/F_MED_CoralNumber_FaceAcc_CtrlRig
graphs [
 RigVMFunctionLibrary {links [], nodes []}
 RigVMModel {nodes [
   RigUnit_BeginExecution = &RigUnit_BeginExecution()
   RigUnit_GetRelativeTransformForItem = &RigUnit_GetRelativeTransformForItem(Child=(Type=Bone,Name="head"), Parent=(Type=Null,Name="head_DefaultPose"), bChildInitial=False, bParentInitial=False)
   RigVMFunction_MathQuaternionToEuler = &RigVMFunction_MathQuaternionToEuler(RotationOrder=ZYX, Value=@RigUnit_GetRelativeTransformForItem.RelativeTransform.Rotation)
   Remap = &RigVMFunction_MathDoubleRemap(SourceMaximum=-64, SourceMinimum=-20, TargetMaximum=1, TargetMinimum=0, bClamp=False, Value=@RigVMFunction_MathQuaternionToEuler.Result.Z)
   &RigUnit_ModifyTransforms(
    ItemToModify="((Item=(Type=Bone,Name=\"Mstr_backHair\"),Transform=(Rotation=(X=-0.065,Y=0,Z=0,W=0.9979),Translation=(X=0,Y=1.5482,Z=1.572),Scale3D=(X=1,Y=1,Z=1))),(Item=(Type=Bone,Name=\"Mstr_ponytail\"),Transform=(Rotation=(X=0.0234,Y=0.1542,Z=0.1484,W=0.9766),Translation= ...(truncated)"
    WeightMinimum=0
    Weight=@Remap.Result
    ExecutePin=@RigUnit_BeginExecution.ExecutePin
   )
   > &RigUnit_ModifyTransforms(
    ItemToModify="((Item=(Type=Bone,Name=\"Mstr_backHair\"),Transform=(Rotation=(X=-0.0814,Y=0,Z=0,W=0.9967),Translation=(X=0,Y=2.2862,Z=3.0751),Scale3D=(X=1,Y=1,Z=1))),(Item=(Type=Bone,Name=\"Mstr_ponytail\"),Transform=(Rotation=(X=0,Y=0.257,Z=0,W=0.9664),Translation=(X=0,Y=0, ...(truncated)"
    WeightMinimum=0
    Weight=&RigVMFunction_MathDoubleRemap(SourceMaximum=-100, SourceMinimum=-64, TargetMaximum=0.7, TargetMinimum=0, bClamp=False, Value=@RigVMFunction_MathQuaternionToEuler.Result.Z).Result
   )
  ]}
]
hierarchy [
 root {loc=<0 0 0>, rot=<0 0 0 1>}
 pelvis {parent=root, loc=<0 -0.5301 98.0571>, rot=<0.0127 0.707 -0.0127 -0.707>}
 spine_01 {parent=pelvis, loc=<6.1208 -0.9328 0>, rot=<0 0 0.034 -0.9994>}
 spine_02 {parent=spine_01, loc=<6.5996 -0.2095 0>, rot=<0 0 0.0238 -0.9997>}
 spine_03 {parent=spine_02, loc=<6.4914 -0.1237 0>, rot=<0 0 0.0149 -0.9999>}
 spine_04 {parent=spine_03, loc=<6.9598 0.0794 0>, rot=<0 0 -0.0132 -0.9999>}
 spine_05 {parent=spine_04, loc=<6.9987 0 0>, rot=<0 0 -0.0015 -1>}
 clavicle_l {parent=spine_05, loc=<5.4831 -0.1053 -3.2946>, rot=<0.0465 -0.7289 -0.009 -0.6829>}
 neck_01 {parent=spine_05, loc=<11.3138 0.1529 0>, rot=<0 0 -0.2003 -0.9797>}
 neck_02 {parent=neck_01, loc=<4.5627 0 0>, rot=<0 0 0.0504 -0.9987>}
 head {parent=neck_02, loc=<5.3914 0 0>, rot=<0 0 0.115 -0.9934>}
 FaceAccRoot {parent=head, loc=<0 0 0>, rot=<0 0 0 1>}
 Dyn_earring_r {parent=FaceAccRoot, loc=<0.3004 0.0298 6.3882>, rot=<0.1055 -0.6992 0.0997 -0.7>}
 Dyn_earring_l {parent=FaceAccRoot, loc=<0.3005 0.0299 -6.3882>, rot=<-0.6992 -0.1055 0.7 0.0997>}
 Dyn_strandHair_l {parent=FaceAccRoot, loc=<6.8597 6.1457 -5.5673>, rot=<-0.0755 0.6076 0.7906 0.0079>}
 Dyn_strandHair1_l {parent=Dyn_strandHair_l, loc=<5.5117 0 0.0983>, rot=<-0.0053 -0.1736 -0.083 -0.9813>}
 Dyn_strandHair_front {parent=FaceAccRoot, loc=<7.8212 8.7759 -1.1252>, rot=<-0.1644 -0.9788 -0.0036 0.1219>}
 Dyn_strandHair2_front {parent=Dyn_strandHair_front, loc=<5.0528 0 0>, rot=<0.0478 -0.2504 0.2626 -0.9306>}
 Dyn_strandHair_r {parent=FaceAccRoot, loc=<6.707 8.3407 3.5894>, rot=<0.0255 -0.9529 -0.2224 0.2044>}
 Dyn_strandHair1_r {parent=Dyn_strandHair_r, loc=<3.3825 0 0>, rot=<0 -0.1647 0 -0.9863>}
 Dyn_earlock {parent=FaceAccRoot, loc=<7.2184 5.6807 6.5892>, rot=<0.0772 -0.7773 -0.619 0.0816>}
 Dyn_earlock_A {parent=Dyn_earlock, loc=<6.2794 -0.3075 -1.1052>, rot=<0.1043 -0.1197 0.1267 -0.9792>}
 Dyn_earlock_B {parent=Dyn_earlock, loc=<6.2173 -0.1955 1.0409>, rot=<0.008 0.0901 -0.0879 -0.992>}
 Mstr_backHair {parent=FaceAccRoot, loc=<9.3223 -4.8438 0>, rot=<-0.405 -0.5796 0.405 -0.5796>}
 Mstr_ponytail {parent=Mstr_backHair, loc=<0 -8.6537 -0.3791>, rot=<0.5779 -0.5779 0.4074 0.4074>}
 Dyn_ponytail {parent=Mstr_ponytail, loc=<0 0 0>, rot=<0 0 0 1>}
 Dyn_ponytail_A {parent=Dyn_ponytail, loc=<3.6499 0.6083 -0.0725>, rot=<-0.7311 0.0937 -0.0513 -0.6738>}
 Dyn_ponytail_B {parent=Dyn_ponytail, loc=<3.0145 -1.9417 0.5751>, rot=<-0.5971 0.4108 0.1309 -0.6764>}
 Dyn_ponytail_C {parent=Dyn_ponytail, loc=<3.0377 0.3255 2.6536>, rot=<-0.6194 0.2272 -0.1855 -0.7283>}
 Dyn_ponytail_D {parent=Dyn_ponytail, loc=<2.2515 -1.0528 2.8217>, rot=<0.6868 -0.5385 0.2069 0.4423>}
 clavicle_r {parent=spine_05, loc=<5.4831 -0.1053 3.2946>, rot=<0.7289 0.0465 -0.6829 0.009>}
 head_DefaultPose {parent=spine_05, type=Null, loc=<20.6573 3.5489 0>, rot=<0 0 -0.036 -0.9994>}
]
