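# Physics asset for the CoralNumber female face-accessory mesh, defining secondary-motion bodies for ponytail, hair strands, earlocks and earrings.
# Sphere and capsule bodies follow the Dyn_ bones, alongside shapes for head, upper spine, FaceAccRoot and Mstr_ponytail.
# Strands, earlock and earrings constrain to head; ponytail pieces A–D branch from Dyn_ponytail, which constrains to Mstr_ponytail.
# Every constraint locks twist and sets narrow swing limits.
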PhysicsAsset /BRCosmetics/Characters/Player/Female/Medium/Bodies/F_MED_CoralNumber/Meshes/Parts/F_MED_CoralNumber_FaceAcc_Physics
bodies {
 Dyn_earlock [{shapeType=Sphere, radius=1.1419, shapeName=Dyn_strandHair_l_sphere, center=<3.8171 0 0>}]
 Dyn_earlock_A [{
   shapeType=Capsule
   length=3.8264
   radius=1.1288
   shapeName=Dyn_earlock_B_capsule
   center=<4.2887 0 0.5344>
   rotation=<rot 90 160.5288 160.5288>
  }]
 Dyn_earlock_B [{shapeType=Capsule, length=2.5558, radius=0.7932, shapeName=Dyn_earlock_B_capsule, center=<3.5278 0 0>, rotation=<rot 90 90 90>}]
 Dyn_earring_l [{shapeType=Sphere, radius=0.8746, shapeName=Dyn_earring_r_sphere, center=<0 0 1.6686>}]
 Dyn_earring_r [{shapeType=Sphere, radius=0.8746, shapeName=Dyn_earring_r_sphere, center=<0 0 -1.6686>}]
 Dyn_ponytail [{shapeType=Sphere, radius=3.0776, shapeName=Dyn_ponytail_sphere, center=<8.524 0 0>}]
 Dyn_ponytail_A [{shapeType=Capsule, length=2.7861, radius=1.7658, shapeName=Dyn_ponytail_A_capsule, center=<5.4639 0.0001 0>, rotation=<rot 90 180 179.9992>}]
 Dyn_ponytail_B [{shapeType=Capsule, length=2.7861, radius=1.7658, shapeName=Dyn_ponytail_A_capsule, center=<5.4639 0.0001 0>, rotation=<rot 90 180 179.9992>}]
 Dyn_ponytail_C [{shapeType=Capsule, length=1.1036, radius=3.4483, shapeName=Dyn_ponytail_A_capsule, center=<5.4639 0.0001 0>, rotation=<rot 90 180 179.9992>}]
 Dyn_ponytail_D [{shapeType=Capsule, length=3.485, radius=1.0668, shapeName=Dyn_ponytail_A_capsule, center=<5.4639 0.0001 0>, rotation=<rot 90 180 179.9992>}]
 Dyn_strandHair1_l [{
   shapeType=Capsule
   length=3.9556
   radius=0.7209
   shapeName=Dyn_strandHair1_l_capsule
   center=<3.4781 0 0>
   rotation=<rot -90 179.9209 -179.9209>
  }]
 Dyn_strandHair1_r [{shapeType=Capsule, length=2.5089, radius=0.8245, shapeName=Dyn_strandHair1_r_capsule, center=<1.7167 0 0>, rotation=<rot 90 179.789 179.789>}]
 Dyn_strandHair2_front [{
   shapeType=Capsule
   length=1.557
   radius=0.5907
   shapeName=Dyn_strandHair1_l_capsule
   center=<3.4781 -0.185 0>
   rotation=<rot -75.0001 -179.9991 179.9993>
  }]
 Dyn_strandHair_front [{shapeType=Sphere, radius=1.1419, shapeName=Dyn_strandHair_l_sphere, center=<3.2978 0 0>}]
 Dyn_strandHair_l [{shapeType=Sphere, radius=1.3184, shapeName=Dyn_strandHair_l_sphere, center=<3.2978 -0.5609 0>}]
 Dyn_strandHair_r [{shapeType=Sphere, radius=1.1419, shapeName=Dyn_strandHair_l_sphere, center=<3.2978 0 0>}]
 FaceAccRoot [{shapeType=Sphere, radius=1.4955, shapeName=FaceAccRoot_sphere}]
 Mstr_ponytail [{shapeType=Sphere, radius=1.382, shapeName=Mstr_ponytail_sphere}]
 head [
  {shapeType=Sphere, radius=5.5779, shapeName=head_sphere}
  {shapeType=Capsule, length=4.6321, radius=5.6729, shapeName=head_capsule, center=<1.0843 3.479 0>, rotation=<rot 90 90 90.0004>}
  {shapeType=Capsule, length=4.0673, radius=1.7636, shapeName=head_capsule, center=<1.6337 7.5889 0>, rotation=<rot 0 54.9996 90>}
  {shapeType=Capsule, length=2.2359, radius=2.03, shapeName=head_capsule, center=<2.0817 -0.3207 5.4079>, rotation=<rot -75 -0.0005 -15>}
 ]
 spine_05 [{shapeType=Capsule, length=19.6734, radius=8.2995, shapeName=spine_05_capsule, center=<3.335 0 0>}]
}
constraints [
 {bone1Name=Mstr_ponytail, bone2Name=Dyn_ponytail, swing1LimitDegrees=30, swing2LimitDegrees=30, twistLimitDegrees=45, twistMotion=Locked}
 {bone1Name=Dyn_ponytail_B, bone2Name=Dyn_ponytail, swing1LimitDegrees=10, swing2LimitDegrees=10, twistLimitDegrees=45, twistMotion=Locked}
 {bone1Name=Dyn_ponytail_D, bone2Name=Dyn_ponytail, swing1LimitDegrees=10, swing2LimitDegrees=10, twistLimitDegrees=45, twistMotion=Locked}
 {bone1Name=Dyn_ponytail_C, bone2Name=Dyn_ponytail, swing1LimitDegrees=10, swing2LimitDegrees=10, twistLimitDegrees=45, twistMotion=Locked}
 {bone1Name=Dyn_ponytail_A, bone2Name=Dyn_ponytail, swing1LimitDegrees=10, swing2LimitDegrees=10, twistLimitDegrees=45, twistMotion=Locked}
 {bone1Name=Dyn_strandHair_l, bone2Name=head, swing1LimitDegrees=15, swing2LimitDegrees=15, twistLimitDegrees=45, twistMotion=Locked}
 {
  bone1Name=Dyn_strandHair1_l
  bone2Name=Dyn_strandHair_l
  swing1LimitDegrees=25
  swing2LimitDegrees=25
  twistLimitDegrees=45
  twistMotion=Locked
 }
 {bone1Name=Dyn_strandHair_front, bone2Name=head, swing1LimitDegrees=15, swing2LimitDegrees=10, twistLimitDegrees=45, twistMotion=Locked}
 {
  bone1Name=Dyn_strandHair2_front
  bone2Name=Dyn_strandHair_front
  swing1LimitDegrees=25
  swing2LimitDegrees=25
  twistLimitDegrees=45
  twistMotion=Locked
 }
 {
  bone1Name=Dyn_strandHair1_r
  bone2Name=Dyn_strandHair_r
  swing1LimitDegrees=15
  swing2LimitDegrees=35
  twistLimitDegrees=45
  twistMotion=Locked
 }
 {bone1Name=Dyn_strandHair_r, bone2Name=head, swing1LimitDegrees=15, swing2LimitDegrees=15, twistLimitDegrees=45, twistMotion=Locked}
 {bone1Name=Dyn_earlock_B, bone2Name=Dyn_earlock, swing1LimitDegrees=15, swing2LimitDegrees=25, twistLimitDegrees=45, twistMotion=Locked}
 {bone1Name=Dyn_earlock_A, bone2Name=Dyn_earlock, swing1LimitDegrees=15, swing2LimitDegrees=20, twistLimitDegrees=45, twistMotion=Locked}
 {bone1Name=Dyn_earlock, bone2Name=head, swing1LimitDegrees=15, swing2LimitDegrees=24, twistLimitDegrees=45, twistMotion=Locked}
 {bone1Name=Dyn_earring_l, bone2Name=head, swing1LimitDegrees=30, swing2LimitDegrees=30, twistLimitDegrees=45, twistMotion=Locked}
 {bone1Name=Dyn_earring_r, bone2Name=head, swing1LimitDegrees=30, swing2LimitDegrees=30, twistLimitDegrees=45, twistMotion=Locked}
]
